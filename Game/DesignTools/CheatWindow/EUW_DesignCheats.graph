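# Editor Utility Widget that gives designers a cheat panel for Play-In-Editor sessions: player cheats, resources, weapons, bots, vehicles, time of day, view modes and debug toggles.
# Most controls build a console command string and pass it to ExecuteConsoleCommandOnAllClients, which runs it in each PIE world.
# Picked weapons are given or dropped directly through function calls, using single-property views bound to the widget.
# BugItGo captures the player's location and camera rotation, or replays edited text as a command.
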EditorUtilityWidgetBlueprint /Game/DesignTools/CheatWindow/EUW_DesignCheats
graphs [
 EventGraph {nodes [
   "On Clicked (Spawn_WeaponPicker)"() {is=title K2Node_ComponentBoundEvent}
   > n1 = GetPIEWorlds!(execute, bIncludeDedicatedServer:bool=true) -> (then, ReturnValue:Array of World Object References)
   > GiveItemToAllPlayers!(
    execute
    WorldContextObject:Object Reference="Get (a copy)"(Array:Array of World Object References=@n1) -> (Output:World Object Reference)
    ItemDefinition:Fort World Item Definition Object Reference=Get SingleProperty_Weapon() -> (SingleProperty_Weapon:Fort Weapon Ranged Item Definition Object Reference)
    ItemVariantGuid:Guid Structure
    NumberToGive:int=1
    bNotifyPlayer:bool=true
   ) -> (then)
   n2 = "On Clicked (Events_PlayEventButton)"() {is=title K2Node_ComponentBoundEvent}
   n3 = "GetText (Text Box)"!(self:Text Box Object Reference=Get Events_EventNumber() -> (Events_EventNumber:Editor Utility Editable Text Box Object Reference)) -> (ReturnValue:Text)
   n4 = "GetText (Text Box)"(self=Get Events_Second() -> (Events_Second:Editor Utility Editable Text Box Object Reference))
   ExecuteConsoleCommandOnAllClients!(
    execute=[@n2.then, @n39.OnEnter, @n37.OnEnter]
    self
    Command:String=Append!(A="cheat Ke* PlayEvent ", B="To String (Text)"(InText=@n3), C:String=" ", D:String="To String (Text)"(InText=@n4)) -> (ReturnValue)
   ) -> (then)
   "On Clicked (All25)"() {is=title K2Node_ComponentBoundEvent}
   > n5 = Set InputInteger(InputInteger:String=25) -> (Output_Get:String)
   ExecuteConsoleCommandOnAllClients(
    execute=[@n6.then, @n7.then, @n46.then, @n5.then]
    Command=Append!(A="GiveResources ", B=Get InputInteger() -> (InputInteger:String)) -> (ReturnValue)
   )
   "On Clicked (All50)"() {is=title K2Node_ComponentBoundEvent}
   > n6 = Set InputInteger(InputInteger=50)
   "On Clicked (All100)"() {is=title K2Node_ComponentBoundEvent}
   > n7 = Set InputInteger(InputInteger=100)
   "On Clicked (Wood10)"() {is=title K2Node_ComponentBoundEvent}
   > n8 = Set InputInteger(InputInteger=10)
   ExecuteConsoleCommandOnAllClients(execute=[@n8.then, @n9.then, @n10.then, @n11.then], Command=Append(A="GiveWood ", B=Get InputInteger()))
   "On Clicked (Wood25)"() {is=title K2Node_ComponentBoundEvent}
   > n9 = Set InputInteger(InputInteger=25)
   "On Clicked (Wood50)"() {is=title K2Node_ComponentBoundEvent}
   > n10 = Set InputInteger(InputInteger=50)
   "On Clicked (Wood100)"() {is=title K2Node_ComponentBoundEvent}
   > n11 = Set InputInteger(InputInteger=100)
   "On Clicked (Brick10)"() {is=title K2Node_ComponentBoundEvent}
   > n12 = Set InputInteger(InputInteger=10)
   ExecuteConsoleCommandOnAllClients(execute=[@n12.then, @n13.then, @n14.then, @n15.then], Command=Append(A="GiveStone ", B=Get InputInteger()))
   "On Clicked (Brick25)"() {is=title K2Node_ComponentBoundEvent}
   > n13 = Set InputInteger(InputInteger=25)
   "On Clicked (Brick50)"() {is=title K2Node_ComponentBoundEvent}
   > n14 = Set InputInteger(InputInteger=50)
   "On Clicked (Button_GodMode)"() {is=title K2Node_ComponentBoundEvent}
   > ExecuteConsoleCommandOnAllClients(Command=God)
   "On Clicked (Brick100)"() {is=title K2Node_ComponentBoundEvent}
   > n15 = Set InputInteger(InputInteger=100)
   "On Clicked (Metal10)"() {is=title K2Node_ComponentBoundEvent}
   > n16 = Set InputInteger(InputInteger=10)
   ExecuteConsoleCommandOnAllClients(execute=[@n16.then, @n19.then, @n17.then, @n18.then], Command=Append(A="GiveMetal ", B=Get InputInteger()))
   "On Clicked (Metal25)"() {is=title K2Node_ComponentBoundEvent}
   > n17 = Set InputInteger(InputInteger=25)
   "On Clicked (Metal50)"() {is=title K2Node_ComponentBoundEvent}
   > n18 = Set InputInteger(InputInteger=50)
   "On Clicked (Metal100)"() {is=title K2Node_ComponentBoundEvent}
   > n19 = Set InputInteger(InputInteger=100)
   "On Clicked (Execute_BugItGo)"() {is=title K2Node_ComponentBoundEvent}
   n20 = Get BugItGo_String() -> (BugItGo_String:Editor Utility Editable Text Box Object Reference)
   > n21 = Branch(Condition="Equal Exactly (Text)"!(A:"Text (by ref)"=Get Text(self=@n20), B:"Text (by ref)"=Get BugitLocationOutput() -> (BugitLocationOutput:Text)) -> (ReturnValue:bool))
   > n22 = GetPIEWorlds(bIncludeDedicatedServer=true)
   n23 = GetPlayerController!(WorldContextObject:Object Reference="Get (a copy)"(Array=@n22), PlayerIndex:int=0) -> (ReturnValue:Player Controller Object Reference)
   n24 = GetCameraRotation!(self:Player Camera Manager Object Reference=Get PlayerCameraManager!(self:Player Controller Object Reference=@n23) -> (PlayerCameraManager:Player Camera Manager Object Reference)) -> (ReturnValue_Roll:float, ReturnValue_Pitch:float, ReturnValue_Yaw:float)
   n25 = Append!(
    A="To String (Float)"(InDouble=NormalizeAxis!(Angle:float=@n24.ReturnValue_Pitch) -> (ReturnValue:float))
    B=" "
    C:String="To String (Float)"(InDouble=NormalizeAxis(Angle=@n24.ReturnValue_Yaw))
    D:String=" "
    E:String="To String (Float)"(InDouble=NormalizeAxis(Angle=@n24))
   ) -> (ReturnValue)
   n26 = Get Actor Location!(self=Get Controlled Pawn!(self:Controller Object Reference=@n23) -> (ReturnValue:Pawn Object Reference) {is=title K2Node_CallFunction}) -> (ReturnValue_X:double, ReturnValue_Y:double, ReturnValue_Z:double)
   n27 = Append(
    A="To String (Float)"(InDouble=@n26)
    B=" "
    C="To String (Float)"(InDouble=@n26.ReturnValue_Y)
    D=" "
    E="To String (Float)"(InDouble=@n26.ReturnValue_Z)
   )
   > n28 = Set Text(
    Text="To Text (String)"(InString=Append!(A="BugItGo ", B=@n27, C=" ", D=@n25) -> (ReturnValue))
    self:Text Box Object Reference=Get BugItGo_String()
   )
   > Set BugitLocationOutput(BugitLocationOutput:Text=@n28) -> (Output_Get:Text)
   n29 = Get Text(self=@n20)
   ExecuteConsoleCommandOnAllClients(execute=@n21.else, Command="To String (Text)"(InText=@n29))
   > Set BugitLocationOutput(BugitLocationOutput=@n29)
   "On Clicked (NPC_Idle)"() {is=title K2Node_ComponentBoundEvent}
   > ExecuteConsoleCommandOnAllClients(Command=SpawnBot Idle)
   n30 = "On Value Committed (Utility_TOD_Slider)"!() -> (OutputDelegate, then, InValue:float, CommitMethod:ETextCommit Enum=Default) {is=title K2Node_ComponentBoundEvent}
   > ExecuteConsoleCommandOnAllClients(Command=Append!(A="cheat SetTimeOfDaySpeed ", B="To String (Float)"(InDouble=@n30.InValue)) -> (ReturnValue))
   "On Clicked (Execute_BugIt)"() {is=title K2Node_ComponentBoundEvent}
   n31 = "To String (Text)"(InText=Get Text(self=Get Bugit_Label() -> (Bugit_Label:Editor Utility Editable Text Box Object Reference)))
   > ExecuteConsoleCommandOnAllClients(Command=Append(A="BugIt ", B=@n31))
   n32 = GetProjectSavedDirectory!() -> (ReturnValue:String)
   > Set Text(
    Text="To Text (String)"(InString=Append!(A=@n32, B=Bugit/WindowsEditor/, C:String=@n31) -> (ReturnValue))
    self=Get Bugit_Output() -> (Bugit_Output:Editor Utility Editable Text Box Object Reference)
   )
   "On Clicked (Button_InfiniteAmmo)"() {is=title K2Node_ComponentBoundEvent}
   > ExecuteConsoleCommandOnAllClients(Command=ToggleInfiniteAmmo)
   "On Clicked (Button_SpawnVehicle)"() {is=title K2Node_ComponentBoundEvent}
   n33 = GetPersistentName!(self:Fort Item Definition Object Reference=Get Property_Vehicle() -> (Property_Vehicle:Fort Vehicle Item Definition Object Reference)) -> (ReturnValue:String)
   > ExecuteConsoleCommandOnAllClients(Command=Append!(A="cheat spawnvehicle ", B=Split(SourceString=@n33, InStr=_, SearchDir=FromStart).RightS) -> (ReturnValue))
   "On Clicked (Button_InfiniteConsumables)"() {is=title K2Node_ComponentBoundEvent}
   > ExecuteConsoleCommandOnAllClients(Command=ToggleInfiniteConsumables)
   "On Clicked (Events_Play)"() {is=title K2Node_ComponentBoundEvent}
   > ExecuteConsoleCommandOnAllClients(Command=FreezeFrame)
   n34 = "On Check State Changed (CheckBox_ScreenMessages)"!() -> (OutputDelegate, then, bIsChecked:bool) {is=title K2Node_ComponentBoundEvent}
   > n35 = Branch(Condition=@n34.bIsChecked)
   > ExecuteConsoleCommandOnAllClients(Command=EnableAllScreenMessages)
   ExecuteConsoleCommandOnAllClients(execute=@n35.else, Command=DisableAllScreenMessages)
   n36 = "On Text Committed (Events_EventNumber)"!() -> (OutputDelegate, then, Text:"Text (by ref)", CommitMethod:ETextCommit Enum=Default) {is=title K2Node_ComponentBoundEvent}
   > n37 = Switch on ETextCommit!(execute, Selection:ETextCommit Enum=@n36.CommitMethod) -> (Default:Exec, OnEnter:Exec, OnUserMovedFocus:Exec, OnCleared:Exec) {is=title K2Node_SwitchEnum}
   n38 = "On Text Committed (Events_Second)"!() -> (OutputDelegate, then, Text:"Text (by ref)", CommitMethod:ETextCommit Enum=Default) {is=title K2Node_ComponentBoundEvent}
   > n39 = Switch on ETextCommit(Selection=@n38.CommitMethod)
   "On Clicked (Button_Ghost)"() {is=title K2Node_ComponentBoundEvent}
   > ExecuteConsoleCommandOnAllClients(Command=Ghost)
   "On Clicked (Button_FreeBuild)"() {is=title K2Node_ComponentBoundEvent}
   > ExecuteConsoleCommandOnAllClients(Command=cheat BuildFree)
   "On Clicked (Drop_WeaponPicker)"() {is=title K2Node_ComponentBoundEvent}
   > n40 = GetPIEWorlds(bIncludeDedicatedServer=true)
   n41 = "Get (a copy)"(Array=@n40)
   n42 = Get Controlled Pawn(self=GetPlayerController(WorldContextObject=@n41, PlayerIndex=0))
   n43 = GetActorForwardVector!(self:Actor Object Reference=@n42) -> (ReturnValue:Vector)
   > SpawnPickupInWorld!(
    execute
    WorldContextObject:Object Reference=@n41
    ItemDefinition:Fort World Item Definition Object Reference=Get SingleProperty_Weapon()
    NumberToSpawn:int=1
    Position:Vector=vector + vector(A=Get Actor Location!(self=@n42) -> (ReturnValue:Vector), B=vector * vector(A=@n43, B=100) -> (ReturnValue=~))
    Direction:Vector="0,0,1"
    OverrideMaxStackCount:int=0
    bToss:bool=false
    bRandomRotation:bool=true
    bBlockedFromAutoPickup:bool=false
    PickupInstigatorHandle:int=-1
    SourceType:EFortPickupSourceTypeFlag Enum=Other
    Source:EFortPickupSpawnSource Enum=Unset
    OptionalOwnerPC:Fort Player Controller Object Reference
    bPickupOnlyRelevantToOwner:bool=true
    OptionalOwnerInstigator:Actor Object Reference
   ) -> (then, ReturnValue:Fort Pickup Object Reference)
   n44 = "On Selection Changed (ComboBox_Viewmode)"!() -> (OutputDelegate, then, SelectedItem:String, SelectionType:ESelectInfo Enum) {is=title K2Node_ComponentBoundEvent}
   > ExecuteConsoleCommandOnAllClients(Command=Append(A="viewmode ", B=@n44.SelectedItem))
   n45 = "On Value Changed (Utility_Slomo_Slider)"!() -> (OutputDelegate, then, InValue:float) {is=title K2Node_ComponentBoundEvent}
   > ExecuteConsoleCommandOnAllClients(Command=Append(A="slomo ", B="To String (Float)"(InDouble=@n45.InValue)))
   "On Clicked (Midnight)"() {is=title K2Node_ComponentBoundEvent}
   > ExecuteConsoleCommandOnAllClients(Command=cheat SetTimeOfDay 0000)
   "On Clicked (Morning)"() {is=title K2Node_ComponentBoundEvent}
   > ExecuteConsoleCommandOnAllClients(Command=cheat SetTimeOfDay 0600)
   "On Clicked (NPC_Bot)"() {is=title K2Node_ComponentBoundEvent}
   > ExecuteConsoleCommandOnAllClients(Command=SpawnBot Player)
   "On Clicked (NPC_NPC)"() {is=title K2Node_ComponentBoundEvent}
   > ExecuteConsoleCommandOnAllClients(Command=SpawnBot NPC)
   "On Clicked (Button_DebugPhysics)"() {is=title K2Node_ComponentBoundEvent}
   > ExecuteConsoleCommandOnAllClients(Command=ShowDebug PHYSICS)
   "On Clicked (Spawn_GoodWeapons)"() {is=title K2Node_ComponentBoundEvent}
   > ExecuteConsoleCommandOnAllClients(Command=GiveGoodWeapons)
   "On Clicked (Spawn_AllWeapons)"() {is=title K2Node_ComponentBoundEvent}
   > ExecuteConsoleCommandOnAllClients(Command=GiveAllWeapons)
   "On Clicked (Spawn_Melee)"() {is=title K2Node_ComponentBoundEvent}
   > ExecuteConsoleCommandOnAllClients(Command=GiveAllMelee)
   "On Clicked (Spawn_Pistols)"() {is=title K2Node_ComponentBoundEvent}
   > ExecuteConsoleCommandOnAllClients(Command=GiveAllPistols)
   "On Clicked (Noon)"() {is=title K2Node_ComponentBoundEvent}
   > ExecuteConsoleCommandOnAllClients(Command=cheat SetTimeOfDay 1200)
   "On Clicked (Spawn_Rifles)"() {is=title K2Node_ComponentBoundEvent}
   > ExecuteConsoleCommandOnAllClients(Command=GiveAllRifles)
   "On Clicked (Spawn_Shotguns)"() {is=title K2Node_ComponentBoundEvent}
   > ExecuteConsoleCommandOnAllClients(Command=GiveAllShotguns)
   "On Clicked (Spawn_Ingredients)"() {is=title K2Node_ComponentBoundEvent}
   > ExecuteConsoleCommandOnAllClients(Command=GiveAllIngredients)
   "On Clicked (Spawn_Traps)"() {is=title K2Node_ComponentBoundEvent}
   > ExecuteConsoleCommandOnAllClients(Command=GiveAllTraps)
   "On Clicked (Spawn_Cosumables)"() {is=title K2Node_ComponentBoundEvent}
   > ExecuteConsoleCommandOnAllClients(Command=GiveAllConsumables)
   "On Clicked (Button_InfiniteCooldowns)"() {is=title K2Node_ComponentBoundEvent}
   > ExecuteConsoleCommandOnAllClients(Command=ToggleNoCoolDown)
   "On Clicked (Button_InfiniteStamina)"() {is=title K2Node_ComponentBoundEvent}
   > ExecuteConsoleCommandOnAllClients(Command=ToggleInfiniteStamina)
   "On Clicked (Button_DebugCamera)"() {is=title K2Node_ComponentBoundEvent}
   > ExecuteConsoleCommandOnAllClients(Command=ToggleDebugCamera)
   "On Clicked (Night)"() {is=title K2Node_ComponentBoundEvent}
   > ExecuteConsoleCommandOnAllClients(Command=cheat SetTimeOfDay 2000)
   "On Clicked (All10)"() {is=title K2Node_ComponentBoundEvent}
   > n46 = Set InputInteger(InputInteger=10)
   Event Construct()
   > SetObject(
    self=[
     Get WeaponSingleProperty() -> (WeaponSingleProperty:Single Property View Object Reference)
     Get SingleProperty_Vehicle() -> (SingleProperty_Vehicle:Single Property View Object Reference)
    ]
    NewObject=Self-Reference()
   )
   Event PreConstruct()
  ]}
 ExecuteConsoleCommandOnAllClients {nodes [
   n1 = ExecuteConsoleCommandOnAllClients!() -> (then, Command:String) {is=title K2Node_FunctionEntry}
   > n2 = GetPIEWorlds(bIncludeDedicatedServer=false)
   > n3 = Branch(Condition="integer > integer"(A=Length(TargetArray:Array of World Object References=@n2)))
   > n4 = For Each Loop(Array:Array of World Object References=@n2) -> (Array Element:World Object Reference)
   > ExecuteConsoleCommand(WorldContextObject=@n4, Command=@n1, SpecificPlayer=GetPlayerController(WorldContextObject=@n4, PlayerIndex=@n4.Array Index))
   PrintString(execute=@n3.else, InString=No PIE worlds available to run commands on, Duration=15)
  ]}
]
variables [
 BugitLocationOutput {type=string, default=""}
 Character Parts {type="array<CustomCharacterPart>", default []}
 DefaultCarClass {type=Car_DEFAULT_C, default=null}
 InputInteger {type=string, default=""}
 Local Player Controller {type=FortPlayerController, default=null}
 LocalPlayerPawn {type=FortPlayerPawn}
 New Cosmetic Loadout {type=CosmeticLoadout, default {slots []}}
 OriginalLoadout {
  type=FortAthenaLoadout
  default {
   aura=null
   backpack=null
   bannerColorId=""
   bannerIconId=""
   bass=null
   character=null
   drums=null
   glider=null
   guitar=null
   keytar=null
   loadingScreen=null
   mic=null
   mimosa=null
   musicPack=null
   pickaxe=null
   shoes=null
   skyDiveContrail=null
   characterVariantChannels []
   dances []
   itemWraps []
  }
 }
 Property_Vehicle {type=FortVehicleItemDefinition, default=/Game/Athena/DrivableVehicles/VID_AntelopeVehicle}
 SelectedActors {type="array<Actor>"}
 SingleProperty_Weapon {type=FortWeaponRangedItemDefinition, default=null}
 Skin {type=AthenaCharacterItemDefinition, default=null}
]
widgets {
 info {namedSlotCount=12, rootWidgetClass=&EditorUtilityScrollBox, widgetCount=186}
 widgets [
  EditorUtilityScrollBox_120 {widgetClassPath=&EditorUtilityScrollBox, widget=EditorUtilityScrollBox_120}
  GeneralCheats_Box {widgetClassPath=&VerticalBox, parent=EditorUtilityScrollBox_120, widget=GeneralCheats_Box, slot {class=ScrollBoxSlot}}
  Separator_5 {widgetClassPath=&HorizontalBox, parent=GeneralCheats_Box, widget=Separator_5, slot {class=VerticalBoxSlot}}
  Border_4 {
   widgetClassPath=&Border
   parent=Separator_5
   widget=Border_4
   slot {class=HorizontalBoxSlot, properties {size {sizeRule=Fill, value=1}}}
   properties=~
  }
  TextBlock_178 {
   widgetClassPath=&TextBlock
   parent=GeneralCheats_Box
   widget=TextBlock_178
   properties {
    text=General
    font {
     bForceMonospaced=false
     bMaterialIsStencil=false
     fontMaterial=null
     fontObject=&Roboto
     letterSpacing=0
     monospacedWidth=1
     size=18
     skewAmount=0
     typefaceFontName=Bold
     outlineSettings {
      bApplyOutlineToDropShadows=false
      bMiteredCorners=false
      bSeparateFillAlpha=false
      outlineMaterial=null
      outlineSize=0
      outlineColor=<rgba 0 0 0 1>
     }
    }
   }
   slot {class=VerticalBoxSlot, properties {horizontalAlignment=HAlign_Center}}
  }
  Button_GodMode {
   widgetClassPath=&EditorUtilityButton
   parent=GeneralCheats_Box
   widget=Button_GodMode
   properties {toolTipText="Turns on God Mode, making you invulnerable."}
   slot {class=VerticalBoxSlot, properties {padding=<ltrb 2.5 0 2.5 2.5>}}
  }
  buttonTitle_5 {
   widgetClassPath=&TextBlock
   parent=Button_GodMode
   widget=buttonTitle_5
   properties {
    justification=Center
    text=God Mode
    font {
     bForceMonospaced=false
     bMaterialIsStencil=false
     fontMaterial=null
     fontObject=&Roboto
     letterSpacing=0
     monospacedWidth=1
     size=9
     skewAmount=0
     typefaceFontName=Regular
     outlineSettings {
      bApplyOutlineToDropShadows=false
      bMiteredCorners=false
      bSeparateFillAlpha=false
      outlineMaterial=null
      outlineSize=0
      outlineColor=<rgba 0 0 0 1>
     }
    }
   }
   slot {class=ButtonSlot}
  }
  Commands {
   widgetClassPath=&HorizontalBox
   parent=GeneralCheats_Box
   widget=Commands
   slot {class=VerticalBoxSlot, properties {size {sizeRule=Fill, value=1}}}
  }
  LeftColumn {
   widgetClassPath=&VerticalBox
   parent=Commands
   widget=LeftColumn
   slot {class=HorizontalBoxSlot, properties {size {sizeRule=Fill, value=1}}}
  }
  Button_Ghost {
   widgetClassPath=&EditorUtilityButton
   parent=LeftColumn
   widget=Button_Ghost
   properties {toolTipText=Allows you to use Fly controls and ignore collision.}
   slot {class=VerticalBoxSlot, properties {padding=<ltrb 2.5 0 2.5 2.5>}}
  }
  buttonTitle_6 {
   widgetClassPath=&TextBlock
   parent=Button_Ghost
   widget=buttonTitle_6
   properties {
    justification=Center
    text=Ghost
    font {
     bForceMonospaced=false
     bMaterialIsStencil=false
     fontMaterial=null
     fontObject=&Roboto
     letterSpacing=0
     monospacedWidth=1
     size=9
     skewAmount=0
     typefaceFontName=Regular
     outlineSettings {
      bApplyOutlineToDropShadows=false
      bMiteredCorners=false
      bSeparateFillAlpha=false
      outlineMaterial=null
      outlineSize=0
      outlineColor=<rgba 0 0 0 1>
     }
    }
   }
   slot {class=ButtonSlot}
  }
  Button_InfiniteAmmo {
   widgetClassPath=&EditorUtilityButton
   parent=LeftColumn
   widget=Button_InfiniteAmmo
   properties {toolTipText="Gives the player infinite ammo. This is not the same as infinite magazine; you still have to reload."}
   slot {class=VerticalBoxSlot, properties {padding=<ltrb 2.5 0 2.5 2.5>}}
  }
  buttonTitle_1 {
   widgetClassPath=&TextBlock
   parent=Button_InfiniteAmmo
   widget=buttonTitle_1
   properties {
    justification=Center
    text=Infinite Ammo
    font {
     bForceMonospaced=false
     bMaterialIsStencil=false
     fontMaterial=null
     fontObject=&Roboto
     letterSpacing=0
     monospacedWidth=1
     size=9
     skewAmount=0
     typefaceFontName=Regular
     outlineSettings {
      bApplyOutlineToDropShadows=false
      bMiteredCorners=false
      bSeparateFillAlpha=false
      outlineMaterial=null
      outlineSize=0
      outlineColor=<rgba 0 0 0 1>
     }
    }
   }
   slot {class=ButtonSlot}
  }
  Button_InfiniteCooldowns {
   widgetClassPath=&EditorUtilityButton
   parent=LeftColumn
   widget=Button_InfiniteCooldowns
   properties {toolTipText=Removes cooldowns and ignores them for future uses.}
   slot {class=VerticalBoxSlot, properties {padding=<ltrb 2.5 0 2.5 2.5>}}
  }
  buttonTitle_2 {
   widgetClassPath=&TextBlock
   parent=Button_InfiniteCooldowns
   widget=buttonTitle_2
   properties {
    justification=Center
    text=Infinite Cooldowns
    font {
     bForceMonospaced=false
     bMaterialIsStencil=false
     fontMaterial=null
     fontObject=&Roboto
     letterSpacing=0
     monospacedWidth=1
     size=9
     skewAmount=0
     typefaceFontName=Regular
     outlineSettings {
      bApplyOutlineToDropShadows=false
      bMiteredCorners=false
      bSeparateFillAlpha=false
      outlineMaterial=null
      outlineSize=0
      outlineColor=<rgba 0 0 0 1>
     }
    }
   }
   slot {class=ButtonSlot}
  }
  RightColumn {
   widgetClassPath=&VerticalBox
   parent=Commands
   widget=RightColumn
   slot {class=HorizontalBoxSlot, properties {size {sizeRule=Fill, value=1}}}
  }
  Button_FreeBuild {
   widgetClassPath=&EditorUtilityButton
   parent=RightColumn
   widget=Button_FreeBuild
   properties {toolTipText=Allows you to build without using resources.}
   slot {class=VerticalBoxSlot, properties {padding=<ltrb 2.5 0 2.5 2.5>}}
  }
  buttonTitle_7 {
   widgetClassPath=&TextBlock
   parent=Button_FreeBuild
   widget=buttonTitle_7
   properties {
    justification=Center
    text=Free Build
    font {
     bForceMonospaced=false
     bMaterialIsStencil=false
     fontMaterial=null
     fontObject=&Roboto
     letterSpacing=0
     monospacedWidth=1
     size=9
     skewAmount=0
     typefaceFontName=Regular
     outlineSettings {
      bApplyOutlineToDropShadows=false
      bMiteredCorners=false
      bSeparateFillAlpha=false
      outlineMaterial=null
      outlineSize=0
      outlineColor=<rgba 0 0 0 1>
     }
    }
   }
   slot {class=ButtonSlot}
  }
  Button_InfiniteConsumables {
   widgetClassPath=&EditorUtilityButton
   parent=RightColumn
   widget=Button_InfiniteConsumables
   properties {toolTipText=Gives you infinite consumables.}
   slot {class=VerticalBoxSlot, properties {padding=<ltrb 2.5 0 2.5 2.5>}}
  }
  buttonTitle_3 {
   widgetClassPath=&TextBlock
   parent=Button_InfiniteConsumables
   widget=buttonTitle_3
   properties {
    justification=Center
    text=Infinite Consumables
    font {
     bForceMonospaced=false
     bMaterialIsStencil=false
     fontMaterial=null
     fontObject=&Roboto
     letterSpacing=0
     monospacedWidth=1
     size=9
     skewAmount=0
     typefaceFontName=Regular
     outlineSettings {
      bApplyOutlineToDropShadows=false
      bMiteredCorners=false
      bSeparateFillAlpha=false
      outlineMaterial=null
      outlineSize=0
      outlineColor=<rgba 0 0 0 1>
     }
    }
   }
   slot {class=ButtonSlot}
  }
  Button_InfiniteStamina {
   widgetClassPath=&EditorUtilityButton
   parent=RightColumn
   widget=Button_InfiniteStamina
   properties {toolTipText=Gives you infinite stamina.}
   slot {class=VerticalBoxSlot, properties {padding=<ltrb 2.5 0 2.5 2.5>}}
  }
  buttonTitle_4 {
   widgetClassPath=&TextBlock
   parent=Button_InfiniteStamina
   widget=buttonTitle_4
   properties {
    justification=Center
    text=Infinite Stamina
    font {
     bForceMonospaced=false
     bMaterialIsStencil=false
     fontMaterial=null
     fontObject=&Roboto
     letterSpacing=0
     monospacedWidth=1
     size=9
     skewAmount=0
     typefaceFontName=Regular
     outlineSettings {
      bApplyOutlineToDropShadows=false
      bMiteredCorners=false
      bSeparateFillAlpha=false
      outlineMaterial=null
      outlineSize=0
      outlineColor=<rgba 0 0 0 1>
     }
    }
   }
   slot {class=ButtonSlot}
  }
  Button_DebugCamera {
   widgetClassPath=&EditorUtilityButton
   parent=GeneralCheats_Box
   widget=Button_DebugCamera
   properties {toolTipText=Dettaches the camera from the player and allows you to slew around the world. This also enables targetting controls for debugging.}
   slot {class=VerticalBoxSlot, properties {padding=<ltrb 2.5 0 2.5 5>}}
  }
  buttonTitle {
   widgetClassPath=&TextBlock
   parent=Button_DebugCamera
   widget=buttonTitle
   properties {
    text=Enable Debug Camera
    font {
     bForceMonospaced=false
     bMaterialIsStencil=false
     fontMaterial=null
     fontObject=&Roboto
     letterSpacing=0
     monospacedWidth=1
     size=9
     skewAmount=0
     typefaceFontName=Regular
     outlineSettings {
      bApplyOutlineToDropShadows=false
      bMiteredCorners=false
      bSeparateFillAlpha=false
      outlineMaterial=null
      outlineSize=0
      outlineColor=<rgba 0 0 0 1>
     }
    }
   }
   slot {class=ButtonSlot}
  }
  TOD_Box {
   widgetClassPath=&HorizontalBox
   parent=GeneralCheats_Box
   widget=TOD_Box
   slot {class=VerticalBoxSlot, properties {padding=<ltrb 0 0 0 2.5>}}
  }
  Exp_TOD {
   widgetClassPath=/Script/Blutility.EditorUtilityExpandableArea
   bIsVariable=true
   parent=TOD_Box
   widget=Exp_TOD
   properties {toolTipText=Time of day controls. Here you can change the time and ToD speed modifier.}
   slot {class=HorizontalBoxSlot, properties {size {sizeRule=Fill, value=1}}}
  }
  Title_1 {
   widgetClassPath=&TextBlock
   namedSlotHost=Exp_TOD
   widget=Title_1
   properties {
    justification=Center
    text=Time of Day
    font {
     bForceMonospaced=false
     bMaterialIsStencil=false
     fontMaterial=null
     fontObject=&Roboto
     letterSpacing=0
     monospacedWidth=1
     size=9
     skewAmount=0
     typefaceFontName=Bold
     outlineSettings {
      bApplyOutlineToDropShadows=false
      bMiteredCorners=false
      bSeparateFillAlpha=false
      outlineMaterial=null
      outlineSize=0
      outlineColor=<rgba 0 0 0 1>
     }
    }
   }
  }
  VerticalBox_308 {widgetClassPath=&VerticalBox, namedSlotHost=Exp_TOD, widget=VerticalBox_308}
  TOD {
   widgetClassPath=&HorizontalBox
   parent=VerticalBox_308
   widget=TOD
   slot {class=VerticalBoxSlot, properties {padding=<ltrb 0 0 0 2.5>}}
  }
  Midnight {
   widgetClassPath=&EditorUtilityButton
   parent=TOD
   widget=Midnight
   properties {toolTipText="Change time of day to 00:00. (12:00AM)"}
   slot {class=HorizontalBoxSlot, properties {size {sizeRule=Fill, value=1}}}
  }
  TextBlock_85 {
   widgetClassPath=&TextBlock
   parent=Midnight
   widget=TextBlock_85
   properties {
    justification=Center
    text=Midnight
    font {
     bForceMonospaced=false
     bMaterialIsStencil=false
     fontMaterial=null
     fontObject=&Roboto
     letterSpacing=0
     monospacedWidth=1
     size=9
     skewAmount=0
     typefaceFontName=Regular
     outlineSettings {
      bApplyOutlineToDropShadows=false
      bMiteredCorners=false
      bSeparateFillAlpha=false
      outlineMaterial=null
      outlineSize=0
      outlineColor=<rgba 0 0 0 1>
     }
    }
   }
   slot {class=ButtonSlot, properties {padding=<ltrb 0 0 0 0>}}
  }
  Morning {
   widgetClassPath=&EditorUtilityButton
   parent=TOD
   widget=Morning
   properties {toolTipText="Change time of day to 06:00. (06:00AM)"}
   slot {class=HorizontalBoxSlot, properties {size {sizeRule=Fill, value=1}}}
  }
  TextBlock_10 {
   widgetClassPath=&TextBlock
   parent=Morning
   widget=TextBlock_10
   properties {
    justification=Center
    text=Morning
    font {
     bForceMonospaced=false
     bMaterialIsStencil=false
     fontMaterial=null
     fontObject=&Roboto
     letterSpacing=0
     monospacedWidth=1
     size=9
     skewAmount=0
     typefaceFontName=Regular
     outlineSettings {
      bApplyOutlineToDropShadows=false
      bMiteredCorners=false
      bSeparateFillAlpha=false
      outlineMaterial=null
      outlineSize=0
      outlineColor=<rgba 0 0 0 1>
     }
    }
   }
   slot {class=ButtonSlot, properties {padding=<ltrb 0 0 0 0>}}
  }
  Noon {
   widgetClassPath=&EditorUtilityButton
   parent=TOD
   widget=Noon
   properties {toolTipText="Change time of day to 12:00. (12:00PM)"}
   slot {class=HorizontalBoxSlot, properties {size {sizeRule=Fill, value=1}}}
  }
  TextBlock_11 {
   widgetClassPath=&TextBlock
   parent=Noon
   widget=TextBlock_11
   properties {
    justification=Center
    text=Noon
    font {
     bForceMonospaced=false
     bMaterialIsStencil=false
     fontMaterial=null
     fontObject=&Roboto
     letterSpacing=0
     monospacedWidth=1
     size=9
     skewAmount=0
     typefaceFontName=Regular
     outlineSettings {
      bApplyOutlineToDropShadows=false
      bMiteredCorners=false
      bSeparateFillAlpha=false
      outlineMaterial=null
      outlineSize=0
      outlineColor=<rgba 0 0 0 1>
     }
    }
   }
   slot {class=ButtonSlot, properties {padding=<ltrb 0 0 0 0>}}
  }
  Night {
   widgetClassPath=&EditorUtilityButton
   parent=TOD
   widget=Night
   properties {toolTipText="Change time of day to 20:00. (08:00PM)"}
   slot {class=HorizontalBoxSlot, properties {size {sizeRule=Fill, value=1}}}
  }
  TextBlock_12 {
   widgetClassPath=&TextBlock
   parent=Night
   widget=TextBlock_12
   properties {
    justification=Center
    text=Night
    font {
     bForceMonospaced=false
     bMaterialIsStencil=false
     fontMaterial=null
     fontObject=&Roboto
     letterSpacing=0
     monospacedWidth=1
     size=9
     skewAmount=0
     typefaceFontName=Regular
     outlineSettings {
      bApplyOutlineToDropShadows=false
      bMiteredCorners=false
      bSeparateFillAlpha=false
      outlineMaterial=null
      outlineSize=0
      outlineColor=<rgba 0 0 0 1>
     }
    }
   }
   slot {class=ButtonSlot, properties {padding=<ltrb 0 0 0 0>}}
  }
  Title_2 {
   widgetClassPath=&TextBlock
   parent=VerticalBox_308
   widget=Title_2
   properties {
    justification=Center
    text=Time of Day Speed
    font {
     bForceMonospaced=false
     bMaterialIsStencil=false
     fontMaterial=null
     fontObject=&Roboto
     letterSpacing=0
     monospacedWidth=1
     size=9
     skewAmount=0
     typefaceFontName=Regular
     outlineSettings {
      bApplyOutlineToDropShadows=false
      bMiteredCorners=false
      bSeparateFillAlpha=false
      outlineMaterial=null
      outlineSize=0
      outlineColor=<rgba 0 0 0 1>
     }
    }
   }
   slot {class=VerticalBoxSlot}
  }
  TODSpeed {widgetClassPath=&HorizontalBox, parent=VerticalBox_308, widget=TODSpeed, slot {class=VerticalBoxSlot}}
  Utility_TOD_Slider {
   widgetClassPath=/Script/Blutility.EditorUtilitySpinBox
   bIsVariable=true
   parent=TODSpeed
   widget=Utility_TOD_Slider
   properties {
    bOverride_MaxSliderValue=true
    bOverride_MaxValue=true
    bOverride_MinSliderValue=true
    bOverride_MinValue=true
    delta=0.25
    maxSliderValue=4
    maxValue=4
    toolTipText=This slider changes the speed modifier for Time of Day.
    value=1
    foregroundColor {colorUseRule=UseColor_Specified, specifiedColor=<rgba 1 1 1 1>}
   }
   slot {class=HorizontalBoxSlot, properties {size {sizeRule=Fill, value=1}}}
  }
  Build {
   widgetClassPath=&HorizontalBox
   parent=GeneralCheats_Box
   widget=Build
   slot {class=VerticalBoxSlot, properties {padding=<ltrb 0 0 0 5>}}
  }
  Exp_Building {
   widgetClassPath=&EditorUtilityExpandableArea
   bIsVariable=true
   parent=Build
   widget=Exp_Building
   properties {toolTipText=Building cheats to give yourself materials and activate free build.}
   slot {class=HorizontalBoxSlot, properties {size {sizeRule=Fill, value=1}}}
  }
  Title {
   widgetClassPath=&TextBlock
   namedSlotHost=Exp_Building
   widget=Title
   properties {
    justification=Center
    text=Resources
    font {
     bForceMonospaced=false
     bMaterialIsStencil=false
     fontMaterial=null
     fontObject=&Roboto
     letterSpacing=0
     monospacedWidth=1
     size=9
     skewAmount=0
     typefaceFontName=Bold
     outlineSettings {
      bApplyOutlineToDropShadows=false
      bMiteredCorners=false
      bSeparateFillAlpha=false
      outlineMaterial=null
      outlineSize=0
      outlineColor=<rgba 0 0 0 1>
     }
    }
   }
  }
  VerticalBox_93 {widgetClassPath=&VerticalBox, namedSlotHost=Exp_Building, widget=VerticalBox_93}
  Resources {
   widgetClassPath=&HorizontalBox
   parent=VerticalBox_93
   widget=Resources
   slot {class=VerticalBoxSlot, properties {padding=<ltrb 0 0 0 10>}}
  }
  TextBlock_14 {
   widgetClassPath=&TextBlock
   parent=Resources
   widget=TextBlock_14
   properties {
    justification=Center
    text=All
    font {
     bForceMonospaced=false
     bMaterialIsStencil=false
     fontMaterial=null
     fontObject=&Roboto
     letterSpacing=0
     monospacedWidth=1
     size=9
     skewAmount=0
     typefaceFontName=Regular
     outlineSettings {
      bApplyOutlineToDropShadows=false
      bMiteredCorners=false
      bSeparateFillAlpha=false
      outlineMaterial=null
      outlineSize=0
      outlineColor=<rgba 0 0 0 1>
     }
    }
   }
   slot {class=HorizontalBoxSlot, properties {verticalAlignment=VAlign_Center, size {sizeRule=Fill, value=1.25}}}
  }
  All10 {
   widgetClassPath=&EditorUtilityButton
   parent=Resources
   widget=All10
   properties {toolTipText=Give yourself 10 of each resource.}
   slot {class=HorizontalBoxSlot, properties {size {sizeRule=Fill, value=1}}}
  }
  Button_All_1 {
   widgetClassPath=&TextBlock
   parent=All10
   widget=Button_All_1
   properties {
    justification=Center
    text=10
    font {
     bForceMonospaced=false
     bMaterialIsStencil=false
     fontMaterial=null
     fontObject=&Roboto
     letterSpacing=0
     monospacedWidth=1
     size=9
     skewAmount=0
     typefaceFontName=Regular
     outlineSettings {
      bApplyOutlineToDropShadows=false
      bMiteredCorners=false
      bSeparateFillAlpha=false
      outlineMaterial=null
      outlineSize=0
      outlineColor=<rgba 0 0 0 1>
     }
    }
   }
   slot {class=ButtonSlot, properties {padding=<ltrb 0 0 0 0>}}
  }
  All25 {
   widgetClassPath=&EditorUtilityButton
   parent=Resources
   widget=All25
   properties {toolTipText=Give yourself 25 of each resource.}
   slot {class=HorizontalBoxSlot, properties {size {sizeRule=Fill, value=1}}}
  }
  Button_All_2 {
   widgetClassPath=&TextBlock
   parent=All25
   widget=Button_All_2
   properties {
    justification=Center
    text=25
    font {
     bForceMonospaced=false
     bMaterialIsStencil=false
     fontMaterial=null
     fontObject=&Roboto
     letterSpacing=0
     monospacedWidth=1
     size=9
     skewAmount=0
     typefaceFontName=Regular
     outlineSettings {
      bApplyOutlineToDropShadows=false
      bMiteredCorners=false
      bSeparateFillAlpha=false
      outlineMaterial=null
      outlineSize=0
      outlineColor=<rgba 0 0 0 1>
     }
    }
   }
   slot {class=ButtonSlot, properties {padding=<ltrb 0 0 0 0>}}
  }
  All50 {
   widgetClassPath=&EditorUtilityButton
   parent=Resources
   widget=All50
   properties {toolTipText=Give yourself 50 of each resource.}
   slot {class=HorizontalBoxSlot, properties {size {sizeRule=Fill, value=1}}}
  }
  Button_All_3 {
   widgetClassPath=&TextBlock
   parent=All50
   widget=Button_All_3
   properties {
    justification=Center
    text=50
    font {
     bForceMonospaced=false
     bMaterialIsStencil=false
     fontMaterial=null
     fontObject=&Roboto
     letterSpacing=0
     monospacedWidth=1
     size=9
     skewAmount=0
     typefaceFontName=Regular
     outlineSettings {
      bApplyOutlineToDropShadows=false
      bMiteredCorners=false
      bSeparateFillAlpha=false
      outlineMaterial=null
      outlineSize=0
      outlineColor=<rgba 0 0 0 1>
     }
    }
   }
   slot {class=ButtonSlot, properties {padding=<ltrb 0 0 0 0>}}
  }
  All100 {
   widgetClassPath=&EditorUtilityButton
   parent=Resources
   widget=All100
   properties {toolTipText=Give yourself 100 of each resource.}
   slot {class=HorizontalBoxSlot, properties {size {sizeRule=Fill, value=1}}}
  }
  Button_All {
   widgetClassPath=&TextBlock
   parent=All100
   widget=Button_All
   properties {
    justification=Center
    text=100
    font {
     bForceMonospaced=false
     bMaterialIsStencil=false
     fontMaterial=null
     fontObject=&Roboto
     letterSpacing=0
     monospacedWidth=1
     size=9
     skewAmount=0
     typefaceFontName=Regular
     outlineSettings {
      bApplyOutlineToDropShadows=false
      bMiteredCorners=false
      bSeparateFillAlpha=false
      outlineMaterial=null
      outlineSize=0
      outlineColor=<rgba 0 0 0 1>
     }
    }
   }
   slot {class=ButtonSlot, properties {padding=<ltrb 0 0 0 0>}}
  }
  Wood {
   widgetClassPath=&HorizontalBox
   parent=VerticalBox_93
   widget=Wood
   slot {class=VerticalBoxSlot, properties {padding=<ltrb 0 0 0 2.5>}}
  }
  TextBlock_13 {
   widgetClassPath=&TextBlock
   parent=Wood
   widget=TextBlock_13
   properties {
    justification=Center
    text=Wood
    font {
     bForceMonospaced=false
     bMaterialIsStencil=false
     fontMaterial=null
     fontObject=&Roboto
     letterSpacing=0
     monospacedWidth=1
     size=9
     skewAmount=0
     typefaceFontName=Regular
     outlineSettings {
      bApplyOutlineToDropShadows=false
      bMiteredCorners=false
      bSeparateFillAlpha=false
      outlineMaterial=null
      outlineSize=0
      outlineColor=<rgba 0 0 0 1>
     }
    }
   }
   slot {class=HorizontalBoxSlot, properties {verticalAlignment=VAlign_Center, size {sizeRule=Fill, value=1.25}}}
  }
  Wood10 {
   widgetClassPath=&EditorUtilityButton
   parent=Wood
   widget=Wood10
   properties {toolTipText=Give yourself 10 Wood.}
   slot {class=HorizontalBoxSlot, properties {size {sizeRule=Fill, value=1}}}
  }
  Button_All_4 {
   widgetClassPath=&TextBlock
   parent=Wood10
   widget=Button_All_4
   properties {
    justification=Center
    text=10
    font {
     bForceMonospaced=false
     bMaterialIsStencil=false
     fontMaterial=null
     fontObject=&Roboto
     letterSpacing=0
     monospacedWidth=1
     size=9
     skewAmount=0
     typefaceFontName=Regular
     outlineSettings {
      bApplyOutlineToDropShadows=false
      bMiteredCorners=false
      bSeparateFillAlpha=false
      outlineMaterial=null
      outlineSize=0
      outlineColor=<rgba 0 0 0 1>
     }
    }
   }
   slot {class=ButtonSlot, properties {padding=<ltrb 0 0 0 0>}}
  }
  Wood25 {
   widgetClassPath=&EditorUtilityButton
   parent=Wood
   widget=Wood25
   properties {toolTipText=Give yourself 25 Wood.}
   slot {class=HorizontalBoxSlot, properties {size {sizeRule=Fill, value=1}}}
  }
  Button_All_5 {
   widgetClassPath=&TextBlock
   parent=Wood25
   widget=Button_All_5
   properties {
    justification=Center
    text=25
    font {
     bForceMonospaced=false
     bMaterialIsStencil=false
     fontMaterial=null
     fontObject=&Roboto
     letterSpacing=0
     monospacedWidth=1
     size=9
     skewAmount=0
     typefaceFontName=Regular
     outlineSettings {
      bApplyOutlineToDropShadows=false
      bMiteredCorners=false
      bSeparateFillAlpha=false
      outlineMaterial=null
      outlineSize=0
      outlineColor=<rgba 0 0 0 1>
     }
    }
   }
   slot {class=ButtonSlot, properties {padding=<ltrb 0 0 0 0>}}
  }
  Wood50 {
   widgetClassPath=&EditorUtilityButton
   parent=Wood
   widget=Wood50
   properties {toolTipText=Give yourself 50 Wood.}
   slot {class=HorizontalBoxSlot, properties {size {sizeRule=Fill, value=1}}}
  }
  Button_All_6 {
   widgetClassPath=&TextBlock
   parent=Wood50
   widget=Button_All_6
   properties {
    justification=Center
    text=50
    font {
     bForceMonospaced=false
     bMaterialIsStencil=false
     fontMaterial=null
     fontObject=&Roboto
     letterSpacing=0
     monospacedWidth=1
     size=9
     skewAmount=0
     typefaceFontName=Regular
     outlineSettings {
      bApplyOutlineToDropShadows=false
      bMiteredCorners=false
      bSeparateFillAlpha=false
      outlineMaterial=null
      outlineSize=0
      outlineColor=<rgba 0 0 0 1>
     }
    }
   }
   slot {class=ButtonSlot, properties {padding=<ltrb 0 0 0 0>}}
  }
  Wood100 {
   widgetClassPath=&EditorUtilityButton
   parent=Wood
   widget=Wood100
   properties {toolTipText=Give yourself 100 Wood.}
   slot {class=HorizontalBoxSlot, properties {size {sizeRule=Fill, value=1}}}
  }
  Button_All_7 {
   widgetClassPath=&TextBlock
   parent=Wood100
   widget=Button_All_7
   properties {
    justification=Center
    text=100
    font {
     bForceMonospaced=false
     bMaterialIsStencil=false
     fontMaterial=null
     fontObject=&Roboto
     letterSpacing=0
     monospacedWidth=1
     size=9
     skewAmount=0
     typefaceFontName=Regular
     outlineSettings {
      bApplyOutlineToDropShadows=false
      bMiteredCorners=false
      bSeparateFillAlpha=false
      outlineMaterial=null
      outlineSize=0
      outlineColor=<rgba 0 0 0 1>
     }
    }
   }
   slot {class=ButtonSlot, properties {padding=<ltrb 0 0 0 0>}}
  }
  Brick {
   widgetClassPath=&HorizontalBox
   parent=VerticalBox_93
   widget=Brick
   slot {class=VerticalBoxSlot, properties {padding=<ltrb 0 0 0 2.5>}}
  }
  TextBlock_16 {
   widgetClassPath=&TextBlock
   parent=Brick
   widget=TextBlock_16
   properties {
    justification=Center
    text=Stone
    font {
     bForceMonospaced=false
     bMaterialIsStencil=false
     fontMaterial=null
     fontObject=&Roboto
     letterSpacing=0
     monospacedWidth=1
     size=9
     skewAmount=0
     typefaceFontName=Regular
     outlineSettings {
      bApplyOutlineToDropShadows=false
      bMiteredCorners=false
      bSeparateFillAlpha=false
      outlineMaterial=null
      outlineSize=0
      outlineColor=<rgba 0 0 0 1>
     }
    }
   }
   slot {class=HorizontalBoxSlot, properties {verticalAlignment=VAlign_Center, size {sizeRule=Fill, value=1.25}}}
  }
  Brick10 {
   widgetClassPath=&EditorUtilityButton
   parent=Brick
   widget=Brick10
   properties {toolTipText=Give yourself 10 Stone.}
   slot {class=HorizontalBoxSlot, properties {size {sizeRule=Fill, value=1}}}
  }
  Button_All_12 {
   widgetClassPath=&TextBlock
   parent=Brick10
   widget=Button_All_12
   properties {
    justification=Center
    text=10
    font {
     bForceMonospaced=false
     bMaterialIsStencil=false
     fontMaterial=null
     fontObject=&Roboto
     letterSpacing=0
     monospacedWidth=1
     size=9
     skewAmount=0
     typefaceFontName=Regular
     outlineSettings {
      bApplyOutlineToDropShadows=false
      bMiteredCorners=false
      bSeparateFillAlpha=false
      outlineMaterial=null
      outlineSize=0
      outlineColor=<rgba 0 0 0 1>
     }
    }
   }
   slot {class=ButtonSlot, properties {padding=<ltrb 0 0 0 0>}}
  }
  Brick25 {
   widgetClassPath=&EditorUtilityButton
   parent=Brick
   widget=Brick25
   properties {toolTipText=Give yourself 25 Stone.}
   slot {class=HorizontalBoxSlot, properties {size {sizeRule=Fill, value=1}}}
  }
  Button_All_13 {
   widgetClassPath=&TextBlock
   parent=Brick25
   widget=Button_All_13
   properties {
    justification=Center
    text=25
    font {
     bForceMonospaced=false
     bMaterialIsStencil=false
     fontMaterial=null
     fontObject=&Roboto
     letterSpacing=0
     monospacedWidth=1
     size=9
     skewAmount=0
     typefaceFontName=Regular
     outlineSettings {
      bApplyOutlineToDropShadows=false
      bMiteredCorners=false
      bSeparateFillAlpha=false
      outlineMaterial=null
      outlineSize=0
      outlineColor=<rgba 0 0 0 1>
     }
    }
   }
   slot {class=ButtonSlot, properties {padding=<ltrb 0 0 0 0>}}
  }
  Brick50 {
   widgetClassPath=&EditorUtilityButton
   parent=Brick
   widget=Brick50
   properties {toolTipText=Give yourself 50 Stone.}
   slot {class=HorizontalBoxSlot, properties {size {sizeRule=Fill, value=1}}}
  }
  Button_All_14 {
   widgetClassPath=&TextBlock
   parent=Brick50
   widget=Button_All_14
   properties {
    justification=Center
    text=50
    font {
     bForceMonospaced=false
     bMaterialIsStencil=false
     fontMaterial=null
     fontObject=&Roboto
     letterSpacing=0
     monospacedWidth=1
     size=9
     skewAmount=0
     typefaceFontName=Regular
     outlineSettings {
      bApplyOutlineToDropShadows=false
      bMiteredCorners=false
      bSeparateFillAlpha=false
      outlineMaterial=null
      outlineSize=0
      outlineColor=<rgba 0 0 0 1>
     }
    }
   }
   slot {class=ButtonSlot, properties {padding=<ltrb 0 0 0 0>}}
  }
  Brick100 {
   widgetClassPath=&EditorUtilityButton
   parent=Brick
   widget=Brick100
   properties {toolTipText=Give yourself 100 Stone.}
   slot {class=HorizontalBoxSlot, properties {size {sizeRule=Fill, value=1}}}
  }
  Button_All_15 {
   widgetClassPath=&TextBlock
   parent=Brick100
   widget=Button_All_15
   properties {
    justification=Center
    text=100
    font {
     bForceMonospaced=false
     bMaterialIsStencil=false
     fontMaterial=null
     fontObject=&Roboto
     letterSpacing=0
     monospacedWidth=1
     size=9
     skewAmount=0
     typefaceFontName=Regular
     outlineSettings {
      bApplyOutlineToDropShadows=false
      bMiteredCorners=false
      bSeparateFillAlpha=false
      outlineMaterial=null
      outlineSize=0
      outlineColor=<rgba 0 0 0 1>
     }
    }
   }
   slot {class=ButtonSlot, properties {padding=<ltrb 0 0 0 0>}}
  }
  Metal {
   widgetClassPath=&HorizontalBox
   parent=VerticalBox_93
   widget=Metal
   slot {class=VerticalBoxSlot, properties {padding=<ltrb 0 0 0 5>}}
  }
  TextBlock_15 {
   widgetClassPath=&TextBlock
   parent=Metal
   widget=TextBlock_15
   properties {
    justification=Center
    text=Metal
    font {
     bForceMonospaced=false
     bMaterialIsStencil=false
     fontMaterial=null
     fontObject=&Roboto
     letterSpacing=0
     monospacedWidth=1
     size=9
     skewAmount=0
     typefaceFontName=Regular
     outlineSettings {
      bApplyOutlineToDropShadows=false
      bMiteredCorners=false
      bSeparateFillAlpha=false
      outlineMaterial=null
      outlineSize=0
      outlineColor=<rgba 0 0 0 1>
     }
    }
   }
   slot {class=HorizontalBoxSlot, properties {verticalAlignment=VAlign_Center, size {sizeRule=Fill, value=1.25}}}
  }
  Metal10 {
   widgetClassPath=&EditorUtilityButton
   parent=Metal
   widget=Metal10
   properties {toolTipText=Give yourself 10 Metal.}
   slot {class=HorizontalBoxSlot, properties {size {sizeRule=Fill, value=1}}}
  }
  Button_All_8 {
   widgetClassPath=&TextBlock
   parent=Metal10
   widget=Button_All_8
   properties {
    justification=Center
    text=10
    font {
     bForceMonospaced=false
     bMaterialIsStencil=false
     fontMaterial=null
     fontObject=&Roboto
     letterSpacing=0
     monospacedWidth=1
     size=9
     skewAmount=0
     typefaceFontName=Regular
     outlineSettings {
      bApplyOutlineToDropShadows=false
      bMiteredCorners=false
      bSeparateFillAlpha=false
      outlineMaterial=null
      outlineSize=0
      outlineColor=<rgba 0 0 0 1>
     }
    }
   }
   slot {class=ButtonSlot, properties {padding=<ltrb 0 0 0 0>}}
  }
  Metal25 {
   widgetClassPath=&EditorUtilityButton
   parent=Metal
   widget=Metal25
   properties {toolTipText=Give yourself 25 Metal.}
   slot {class=HorizontalBoxSlot, properties {size {sizeRule=Fill, value=1}}}
  }
  Button_All_9 {
   widgetClassPath=&TextBlock
   parent=Metal25
   widget=Button_All_9
   properties {
    justification=Center
    text=25
    font {
     bForceMonospaced=false
     bMaterialIsStencil=false
     fontMaterial=null
     fontObject=&Roboto
     letterSpacing=0
     monospacedWidth=1
     size=9
     skewAmount=0
     typefaceFontName=Regular
     outlineSettings {
      bApplyOutlineToDropShadows=false
      bMiteredCorners=false
      bSeparateFillAlpha=false
      outlineMaterial=null
      outlineSize=0
      outlineColor=<rgba 0 0 0 1>
     }
    }
   }
   slot {class=ButtonSlot, properties {padding=<ltrb 0 0 0 0>}}
  }
  Metal50 {
   widgetClassPath=&EditorUtilityButton
   parent=Metal
   widget=Metal50
   properties {toolTipText=Give yourself 50 Metal.}
   slot {class=HorizontalBoxSlot, properties {size {sizeRule=Fill, value=1}}}
  }
  Button_All_10 {
   widgetClassPath=&TextBlock
   parent=Metal50
   widget=Button_All_10
   properties {
    justification=Center
    text=50
    font {
     bForceMonospaced=false
     bMaterialIsStencil=false
     fontMaterial=null
     fontObject=&Roboto
     letterSpacing=0
     monospacedWidth=1
     size=9
     skewAmount=0
     typefaceFontName=Regular
     outlineSettings {
      bApplyOutlineToDropShadows=false
      bMiteredCorners=false
      bSeparateFillAlpha=false
      outlineMaterial=null
      outlineSize=0
      outlineColor=<rgba 0 0 0 1>
     }
    }
   }
   slot {class=ButtonSlot, properties {padding=<ltrb 0 0 0 0>}}
  }
  Metal100 {
   widgetClassPath=&EditorUtilityButton
   parent=Metal
   widget=Metal100
   properties {toolTipText=Give yourself 100 Metal.}
   slot {class=HorizontalBoxSlot, properties {size {sizeRule=Fill, value=1}}}
  }
  Button_All_11 {
   widgetClassPath=&TextBlock
   parent=Metal100
   widget=Button_All_11
   properties {
    justification=Center
    text=100
    font {
     bForceMonospaced=false
     bMaterialIsStencil=false
     fontMaterial=null
     fontObject=&Roboto
     letterSpacing=0
     monospacedWidth=1
     size=9
     skewAmount=0
     typefaceFontName=Regular
     outlineSettings {
      bApplyOutlineToDropShadows=false
      bMiteredCorners=false
      bSeparateFillAlpha=false
      outlineMaterial=null
      outlineSize=0
      outlineColor=<rgba 0 0 0 1>
     }
    }
   }
   slot {class=ButtonSlot, properties {padding=<ltrb 0 0 0 0>}}
  }
  Bugit {
   widgetClassPath=&HorizontalBox
   parent=GeneralCheats_Box
   widget=Bugit
   slot {class=VerticalBoxSlot, properties {padding=<ltrb 0 0 0 5>}}
  }
  Exp_BugIt {
   widgetClassPath=&EditorUtilityExpandableArea
   bIsVariable=true
   parent=Bugit
   widget=Exp_BugIt
   properties {toolTipText=Building cheats to give yourself materials and activate free build.}
   slot {class=HorizontalBoxSlot, properties {size {sizeRule=Fill, value=1}}}
  }
  Title_3 {
   widgetClassPath=&TextBlock
   namedSlotHost=Exp_BugIt
   widget=Title_3
   properties {
    justification=Center
    text=Bugit
    font {
     bForceMonospaced=false
     bMaterialIsStencil=false
     fontMaterial=null
     fontObject=&Roboto
     letterSpacing=0
     monospacedWidth=1
     size=9
     skewAmount=0
     typefaceFontName=Bold
     outlineSettings {
      bApplyOutlineToDropShadows=false
      bMiteredCorners=false
      bSeparateFillAlpha=false
      outlineMaterial=null
      outlineSize=0
      outlineColor=<rgba 0 0 0 1>
     }
    }
   }
  }
  VerticalBox {widgetClassPath=&VerticalBox, namedSlotHost=Exp_BugIt, widget=VerticalBox}
  BugItGoCall {
   widgetClassPath=&HorizontalBox
   parent=VerticalBox
   widget=BugItGoCall
   slot {class=VerticalBoxSlot, properties {padding=<ltrb 0 0 0 5>}}
  }
  Execute_BugItGo {
   widgetClassPath=&EditorUtilityButton
   parent=BugItGoCall
   widget=Execute_BugItGo
   properties {toolTipText=Fill input to execute or leave untouched to print command for current location}
   slot {class=HorizontalBoxSlot, properties {padding=<ltrb 2.5 0 0 0>, size {sizeRule=Fill, value=1}}}
  }
  TextBlock_3 {
   widgetClassPath=&TextBlock
   parent=Execute_BugItGo
   widget=TextBlock_3
   properties {
    text=BugItGo
    font {
     bForceMonospaced=false
     bMaterialIsStencil=false
     fontMaterial=null
     fontObject=&Roboto
     letterSpacing=0
     monospacedWidth=1
     size=9
     skewAmount=0
     typefaceFontName=Regular
     outlineSettings {
      bApplyOutlineToDropShadows=false
      bMiteredCorners=false
      bSeparateFillAlpha=false
      outlineMaterial=null
      outlineSize=0
      outlineColor=<rgba 0 0 0 1>
     }
    }
   }
   slot {class=ButtonSlot}
  }
  BugItGo_String {
   widgetClassPath=&EditorUtilityEditableTextBox
   bIsVariable=true
   parent=BugItGoCall
   widget=BugItGo_String
   properties {
    hintText=BugItGo Command
    justification=Center
    keyboardType=Number
    selectAllTextWhenFocused=true
    widgetStyle {
     backgroundColor {colorUseRule=UseColor_Specified, specifiedColor=<rgba 1 1 1 1>}
     backgroundImageFocused {
      bIsDynamicallyLoaded=false
      drawAs=RoundedBox
      imageType=NoImage
      mirroring=NoMirror
      resourceName=None
      resourceObject=null
      tiling=NoTile
      imageSize=<0 0>
      margin=<ltrb 0 0 0 0>
      outlineSettings {
       bUseBrushTransparency=false
       roundingType=FixedRadius
       width=1
       color {colorUseRule=UseColor_ColorTable, specifiedColor=<rgba 0 0 0 0>}
       cornerRadii=<4 4 4 4>
      }
      tintColor {colorUseRule=UseColor_Specified, specifiedColor=<rgba 0.0048 0.0048 0.0048 1>}
      uVRegion {bIsValid=false, max=<0 0>, min=<0 0>}
     }
     backgroundImageHovered {
      bIsDynamicallyLoaded=false
      drawAs=RoundedBox
      imageType=NoImage
      mirroring=NoMirror
      resourceName=None
      resourceObject=null
      tiling=NoTile
      imageSize=<0 0>
      margin=<ltrb 0 0 0 0>
      outlineSettings {
       bUseBrushTransparency=false
       roundingType=FixedRadius
       width=1
       color {colorUseRule=UseColor_ColorTable, specifiedColor=<rgba 0 0 0 0>}
       cornerRadii=<4 4 4 4>
      }
      tintColor {colorUseRule=UseColor_Specified, specifiedColor=<rgba 0.0048 0.0048 0.0048 1>}
      uVRegion {bIsValid=false, max=<0 0>, min=<0 0>}
     }
     backgroundImageNormal {
      bIsDynamicallyLoaded=false
      drawAs=RoundedBox
      imageType=NoImage
      mirroring=NoMirror
      resourceName=None
      resourceObject=null
      tiling=NoTile
      imageSize=<0 0>
      margin=<ltrb 0 0 0 0>
      outlineSettings {
       bUseBrushTransparency=false
       roundingType=FixedRadius
       width=1
       color {colorUseRule=UseColor_ColorTable, specifiedColor=<rgba 0 0 0 0>}
       cornerRadii=<4 4 4 4>
      }
      tintColor {colorUseRule=UseColor_Specified, specifiedColor=<rgba 0.0048 0.0048 0.0048 1>}
      uVRegion {bIsValid=false, max=<0 0>, min=<0 0>}
     }
     backgroundImageReadOnly {
      bIsDynamicallyLoaded=false
      drawAs=RoundedBox
      imageType=NoImage
      mirroring=NoMirror
      resourceName=None
      resourceObject=null
      tiling=NoTile
      imageSize=<0 0>
      margin=<ltrb 0 0 0 0>
      outlineSettings {
       bUseBrushTransparency=false
       roundingType=FixedRadius
       width=1
       color {colorUseRule=UseColor_ColorTable, specifiedColor=<rgba 0 0 0 0>}
       cornerRadii=<4 4 4 4>
      }
      tintColor {colorUseRule=UseColor_Specified, specifiedColor=<rgba 0.0284 0.0284 0.0284 1>}
      uVRegion {bIsValid=false, max=<0 0>, min=<0 0>}
     }
     focusedForegroundColor {colorUseRule=UseColor_Specified, specifiedColor=<rgba 1 1 1 1>}
     foregroundColor {colorUseRule=UseColor_Specified, specifiedColor=<rgba 0.5271 0.5271 0.5271 1>}
     hScrollBarPadding=<ltrb 0 0 0 0>
     padding=<ltrb 12 3 8 4>
     readOnlyForegroundColor {colorUseRule=UseColor_Foreground, specifiedColor=<rgba 1 0 1 1>}
     scrollBarStyle {
      thickness=8
      draggedThumbImage {
       bIsDynamicallyLoaded=false
       drawAs=RoundedBox
       imageType=NoImage
       mirroring=NoMirror
       resourceName=None
       resourceObject=null
       tiling=NoTile
       imageSize=<0 0>
       margin=<ltrb 0 0 0 0>
       outlineSettings {
        bUseBrushTransparency=false
        roundingType=FixedRadius
        width=0
        color {colorUseRule=UseColor_Specified, specifiedColor=<rgba 0 0 0 0>}
        cornerRadii=<4 4 4 4>
       }
       tintColor {colorUseRule=UseColor_Specified, specifiedColor=<rgba 0.2159 0.2159 0.2159 1>}
       uVRegion {bIsValid=false, max=<0 0>, min=<0 0>}
      }
      horizontalBackgroundImage {
       bIsDynamicallyLoaded=false
       drawAs=NoDrawType
       imageType=NoImage
       mirroring=NoMirror
       resourceName=None
       resourceObject=null
       tiling=NoTile
       imageSize=<0 0>
       margin=<ltrb 0 0 0 0>
       outlineSettings {
        bUseBrushTransparency=false
        roundingType=HalfHeightRadius
        width=0
        color {colorUseRule=UseColor_Specified, specifiedColor=<rgba 0 0 0 0>}
        cornerRadii=<0 0 0 0>
       }
       tintColor {colorUseRule=UseColor_Specified, specifiedColor=<rgba 1 1 1 1>}
       uVRegion {bIsValid=false, max=<0 0>, min=<0 0>}
      }
      horizontalBottomSlotImage {
       bIsDynamicallyLoaded=false
       drawAs=NoDrawType
       imageType=NoImage
       mirroring=NoMirror
       resourceName=None
       resourceObject=null
       tiling=NoTile
       imageSize=<0 0>
       margin=<ltrb 0 0 0 0>
       outlineSettings {
        bUseBrushTransparency=false
        roundingType=HalfHeightRadius
        width=0
        color {colorUseRule=UseColor_Specified, specifiedColor=<rgba 0 0 0 0>}
        cornerRadii=<0 0 0 0>
       }
       tintColor {colorUseRule=UseColor_Specified, specifiedColor=<rgba 1 1 1 1>}
       uVRegion {bIsValid=false, max=<0 0>, min=<0 0>}
      }
      horizontalTopSlotImage {
       bIsDynamicallyLoaded=false
       drawAs=NoDrawType
       imageType=NoImage
       mirroring=NoMirror
       resourceName=None
       resourceObject=null
       tiling=NoTile
       imageSize=<0 0>
       margin=<ltrb 0 0 0 0>
       outlineSettings {
        bUseBrushTransparency=false
        roundingType=HalfHeightRadius
        width=0
        color {colorUseRule=UseColor_Specified, specifiedColor=<rgba 0 0 0 0>}
        cornerRadii=<0 0 0 0>
       }
       tintColor {colorUseRule=UseColor_Specified, specifiedColor=<rgba 1 1 1 1>}
       uVRegion {bIsValid=false, max=<0 0>, min=<0 0>}
      }
      hoveredThumbImage {
       bIsDynamicallyLoaded=false
       drawAs=RoundedBox
       imageType=NoImage
       mirroring=NoMirror
       resourceName=None
       resourceObject=null
       tiling=NoTile
       imageSize=<0 0>
       margin=<ltrb 0 0 0 0>
       outlineSettings {
        bUseBrushTransparency=false
        roundingType=FixedRadius
        width=0
        color {colorUseRule=UseColor_Specified, specifiedColor=<rgba 0 0 0 0>}
        cornerRadii=<4 4 4 4>
       }
       tintColor {colorUseRule=UseColor_Specified, specifiedColor=<rgba 0.2159 0.2159 0.2159 1>}
       uVRegion {bIsValid=false, max=<0 0>, min=<0 0>}
      }
      normalThumbImage {
       bIsDynamicallyLoaded=false
       drawAs=RoundedBox
       imageType=NoImage
       mirroring=NoMirror
       resourceName=None
       resourceObject=null
       tiling=NoTile
       imageSize=<0 0>
       margin=<ltrb 0 0 0 0>
       outlineSettings {
        bUseBrushTransparency=false
        roundingType=FixedRadius
        width=0
        color {colorUseRule=UseColor_Specified, specifiedColor=<rgba 0 0 0 0>}
        cornerRadii=<4 4 4 4>
       }
       tintColor {colorUseRule=UseColor_Specified, specifiedColor=<rgba 0.0953 0.0953 0.0953 1>}
       uVRegion {bIsValid=false, max=<0 0>, min=<0 0>}
      }
      verticalBackgroundImage {
       bIsDynamicallyLoaded=false
       drawAs=NoDrawType
       imageType=NoImage
       mirroring=NoMirror
       resourceName=None
       resourceObject=null
       tiling=NoTile
       imageSize=<0 0>
       margin=<ltrb 0 0 0 0>
       outlineSettings {
        bUseBrushTransparency=false
        roundingType=HalfHeightRadius
        width=0
        color {colorUseRule=UseColor_Specified, specifiedColor=<rgba 0 0 0 0>}
        cornerRadii=<0 0 0 0>
       }
       tintColor {colorUseRule=UseColor_Specified, specifiedColor=<rgba 1 1 1 1>}
       uVRegion {bIsValid=false, max=<0 0>, min=<0 0>}
      }
      verticalBottomSlotImage {
       bIsDynamicallyLoaded=false
       drawAs=NoDrawType
       imageType=NoImage
       mirroring=NoMirror
       resourceName=None
       resourceObject=null
       tiling=NoTile
       imageSize=<0 0>
       margin=<ltrb 0 0 0 0>
       outlineSettings {
        bUseBrushTransparency=false
        roundingType=HalfHeightRadius
        width=0
        color {colorUseRule=UseColor_Specified, specifiedColor=<rgba 0 0 0 0>}
        cornerRadii=<0 0 0 0>
       }
       tintColor {colorUseRule=UseColor_Specified, specifiedColor=<rgba 1 1 1 1>}
       uVRegion {bIsValid=false, max=<0 0>, min=<0 0>}
      }
      verticalTopSlotImage {
       bIsDynamicallyLoaded=false
       drawAs=NoDrawType
       imageType=NoImage
       mirroring=NoMirror
       resourceName=None
       resourceObject=null
       tiling=NoTile
       imageSize=<0 0>
       margin=<ltrb 0 0 0 0>
       outlineSettings {
        bUseBrushTransparency=false
        roundingType=HalfHeightRadius
        width=0
        color {colorUseRule=UseColor_Specified, specifiedColor=<rgba 0 0 0 0>}
        cornerRadii=<0 0 0 0>
       }
       tintColor {colorUseRule=UseColor_Specified, specifiedColor=<rgba 1 1 1 1>}
       uVRegion {bIsValid=false, max=<0 0>, min=<0 0>}
      }
     }
     textStyle {
      overflowPolicy=Clip
      transformPolicy=None
      colorAndOpacity {colorUseRule=UseColor_Foreground, specifiedColor=<rgba 1 0 1 1>}
      font {
       bForceMonospaced=false
       bMaterialIsStencil=false
       fontMaterial=null
       fontObject=null
       letterSpacing=0
       monospacedWidth=1
       size=10
       skewAmount=0
       typefaceFontName=Regular
       outlineSettings {
        bApplyOutlineToDropShadows=false
        bMiteredCorners=false
        bSeparateFillAlpha=false
        outlineMaterial=null
        outlineSize=0
        outlineColor=<rgba 0 0 0 1>
       }
      }
      highlightColor {colorUseRule=UseColor_ColorTable, specifiedColor=<rgba 0 0 0 0>}
      highlightShape {
       bIsDynamicallyLoaded=false
       drawAs=Image
       imageType=NoImage
       mirroring=NoMirror
       resourceName=None
       resourceObject=null
       tiling=NoTile
       imageSize=<0 0>
       margin=<ltrb 0 0 0 0>
       outlineSettings {
        bUseBrushTransparency=false
        roundingType=HalfHeightRadius
        width=0
        color {colorUseRule=UseColor_Specified, specifiedColor=<rgba 0 0 0 0>}
        cornerRadii=<0 0 0 0>
       }
       tintColor {colorUseRule=UseColor_ColorTable, specifiedColor=<rgba 0 0 0 0>}
       uVRegion {bIsValid=false, max=<0 0>, min=<0 0>}
      }
      selectedBackgroundColor {colorUseRule=UseColor_ColorTable, specifiedColor=<rgba 0 0 0 0>}
      shadowColorAndOpacity=<rgba 0 0 0 1>
      shadowOffset=<0 0>
      strikeBrush {
       bIsDynamicallyLoaded=false
       drawAs=Image
       imageType=NoImage
       mirroring=NoMirror
       resourceName=None
       resourceObject=null
       tiling=NoTile
       imageSize=<32 32>
       margin=<ltrb 0 0 0 0>
       outlineSettings {
        bUseBrushTransparency=false
        roundingType=HalfHeightRadius
        width=0
        color {colorUseRule=UseColor_Specified, specifiedColor=<rgba 0 0 0 0>}
        cornerRadii=<0 0 0 0>
       }
       tintColor {colorUseRule=UseColor_Specified, specifiedColor=<rgba 1 1 1 1>}
       uVRegion {bIsValid=false, max=<0 0>, min=<0 0>}
      }
      underlineBrush {
       bIsDynamicallyLoaded=false
       drawAs=Image
       imageType=NoImage
       mirroring=NoMirror
       resourceName=None
       resourceObject=null
       tiling=NoTile
       imageSize=<32 32>
       margin=<ltrb 0 0 0 0>
       outlineSettings {
        bUseBrushTransparency=false
        roundingType=HalfHeightRadius
        width=0
        color {colorUseRule=UseColor_Specified, specifiedColor=<rgba 0 0 0 0>}
        cornerRadii=<0 0 0 0>
       }
       tintColor {colorUseRule=UseColor_Specified, specifiedColor=<rgba 1 1 1 1>}
       uVRegion {bIsValid=false, max=<0 0>, min=<0 0>}
      }
     }
     vScrollBarPadding=<ltrb 0 0 0 0>
    }
   }
   slot {class=HorizontalBoxSlot, properties {size {sizeRule=Fill, value=1}}}
  }
  BugItCall {
   widgetClassPath=&HorizontalBox
   parent=VerticalBox
   widget=BugItCall
   slot {class=VerticalBoxSlot, properties {padding=<ltrb 0 0 0 5>}}
  }
  Execute_BugIt {
   widgetClassPath=&EditorUtilityButton
   parent=BugItCall
   widget=Execute_BugIt
   properties {toolTipText=Take BugIt with Label in text box}
   slot {class=HorizontalBoxSlot, properties {padding=<ltrb 2.5 0 0 0>, size {sizeRule=Fill, value=1}}}
  }
  TextBlock_5 {
   widgetClassPath=&TextBlock
   parent=Execute_BugIt
   widget=TextBlock_5
   properties {
    text=BugIt
    font {
     bForceMonospaced=false
     bMaterialIsStencil=false
     fontMaterial=null
     fontObject=&Roboto
     letterSpacing=0
     monospacedWidth=1
     size=9
     skewAmount=0
     typefaceFontName=Regular
     outlineSettings {
      bApplyOutlineToDropShadows=false
      bMiteredCorners=false
      bSeparateFillAlpha=false
      outlineMaterial=null
      outlineSize=0
      outlineColor=<rgba 0 0 0 1>
     }
    }
   }
   slot {class=ButtonSlot}
  }
  Bugit_Label {
   widgetClassPath=&EditorUtilityEditableTextBox
   bIsVariable=true
   parent=BugItCall
   widget=Bugit_Label
   properties {
    hintText=BugIt Name
    justification=Center
    keyboardType=Number
    selectAllTextWhenFocused=true
    text=Label
    widgetStyle {
     backgroundColor {colorUseRule=UseColor_Specified, specifiedColor=<rgba 1 1 1 1>}
     backgroundImageFocused {
      bIsDynamicallyLoaded=false
      drawAs=RoundedBox
      imageType=NoImage
      mirroring=NoMirror
      resourceName=None
      resourceObject=null
      tiling=NoTile
      imageSize=<0 0>
      margin=<ltrb 0 0 0 0>
      outlineSettings {
       bUseBrushTransparency=false
       roundingType=FixedRadius
       width=1
       color {colorUseRule=UseColor_ColorTable, specifiedColor=<rgba 0 0 0 0>}
       cornerRadii=<4 4 4 4>
      }
      tintColor {colorUseRule=UseColor_Specified, specifiedColor=<rgba 0.0048 0.0048 0.0048 1>}
      uVRegion {bIsValid=false, max=<0 0>, min=<0 0>}
     }
     backgroundImageHovered {
      bIsDynamicallyLoaded=false
      drawAs=RoundedBox
      imageType=NoImage
      mirroring=NoMirror
      resourceName=None
      resourceObject=null
      tiling=NoTile
      imageSize=<0 0>
      margin=<ltrb 0 0 0 0>
      outlineSettings {
       bUseBrushTransparency=false
       roundingType=FixedRadius
       width=1
       color {colorUseRule=UseColor_ColorTable, specifiedColor=<rgba 0 0 0 0>}
       cornerRadii=<4 4 4 4>
      }
      tintColor {colorUseRule=UseColor_Specified, specifiedColor=<rgba 0.0048 0.0048 0.0048 1>}
      uVRegion {bIsValid=false, max=<0 0>, min=<0 0>}
     }
     backgroundImageNormal {
      bIsDynamicallyLoaded=false
      drawAs=RoundedBox
      imageType=NoImage
      mirroring=NoMirror
      resourceName=None
      resourceObject=null
      tiling=NoTile
      imageSize=<0 0>
      margin=<ltrb 0 0 0 0>
      outlineSettings {
       bUseBrushTransparency=false
       roundingType=FixedRadius
       width=1
       color {colorUseRule=UseColor_ColorTable, specifiedColor=<rgba 0 0 0 0>}
       cornerRadii=<4 4 4 4>
      }
      tintColor {colorUseRule=UseColor_Specified, specifiedColor=<rgba 0.0048 0.0048 0.0048 1>}
      uVRegion {bIsValid=false, max=<0 0>, min=<0 0>}
     }
     backgroundImageReadOnly {
      bIsDynamicallyLoaded=false
      drawAs=RoundedBox
      imageType=NoImage
      mirroring=NoMirror
      resourceName=None
      resourceObject=null
      tiling=NoTile
      imageSize=<0 0>
      margin=<ltrb 0 0 0 0>
      outlineSettings {
       bUseBrushTransparency=false
       roundingType=FixedRadius
       width=1
       color {colorUseRule=UseColor_ColorTable, specifiedColor=<rgba 0 0 0 0>}
       cornerRadii=<4 4 4 4>
      }
      tintColor {colorUseRule=UseColor_Specified, specifiedColor=<rgba 0.0284 0.0284 0.0284 1>}
      uVRegion {bIsValid=false, max=<0 0>, min=<0 0>}
     }
     focusedForegroundColor {colorUseRule=UseColor_Specified, specifiedColor=<rgba 1 1 1 1>}
     foregroundColor {colorUseRule=UseColor_Specified, specifiedColor=<rgba 0.5271 0.5271 0.5271 1>}
     hScrollBarPadding=<ltrb 0 0 0 0>
     padding=<ltrb 12 3 8 4>
     readOnlyForegroundColor {colorUseRule=UseColor_Foreground, specifiedColor=<rgba 1 0 1 1>}
     scrollBarStyle {
      thickness=8
      draggedThumbImage {
       bIsDynamicallyLoaded=false
       drawAs=RoundedBox
       imageType=NoImage
       mirroring=NoMirror
       resourceName=None
       resourceObject=null
       tiling=NoTile
       imageSize=<0 0>
       margin=<ltrb 0 0 0 0>
       outlineSettings {
        bUseBrushTransparency=false
        roundingType=FixedRadius
        width=0
        color {colorUseRule=UseColor_Specified, specifiedColor=<rgba 0 0 0 0>}
        cornerRadii=<4 4 4 4>
       }
       tintColor {colorUseRule=UseColor_Specified, specifiedColor=<rgba 0.2159 0.2159 0.2159 1>}
       uVRegion {bIsValid=false, max=<0 0>, min=<0 0>}
      }
      horizontalBackgroundImage {
       bIsDynamicallyLoaded=false
       drawAs=NoDrawType
       imageType=NoImage
       mirroring=NoMirror
       resourceName=None
       resourceObject=null
       tiling=NoTile
       imageSize=<0 0>
       margin=<ltrb 0 0 0 0>
       outlineSettings {
        bUseBrushTransparency=false
        roundingType=HalfHeightRadius
        width=0
        color {colorUseRule=UseColor_Specified, specifiedColor=<rgba 0 0 0 0>}
        cornerRadii=<0 0 0 0>
       }
       tintColor {colorUseRule=UseColor_Specified, specifiedColor=<rgba 1 1 1 1>}
       uVRegion {bIsValid=false, max=<0 0>, min=<0 0>}
      }
      horizontalBottomSlotImage {
       bIsDynamicallyLoaded=false
       drawAs=NoDrawType
       imageType=NoImage
       mirroring=NoMirror
       resourceName=None
       resourceObject=null
       tiling=NoTile
       imageSize=<0 0>
       margin=<ltrb 0 0 0 0>
       outlineSettings {
        bUseBrushTransparency=false
        roundingType=HalfHeightRadius
        width=0
        color {colorUseRule=UseColor_Specified, specifiedColor=<rgba 0 0 0 0>}
        cornerRadii=<0 0 0 0>
       }
       tintColor {colorUseRule=UseColor_Specified, specifiedColor=<rgba 1 1 1 1>}
       uVRegion {bIsValid=false, max=<0 0>, min=<0 0>}
      }
      horizontalTopSlotImage {
       bIsDynamicallyLoaded=false
       drawAs=NoDrawType
       imageType=NoImage
       mirroring=NoMirror
       resourceName=None
       resourceObject=null
       tiling=NoTile
       imageSize=<0 0>
       margin=<ltrb 0 0 0 0>
       outlineSettings {
        bUseBrushTransparency=false
        roundingType=HalfHeightRadius
        width=0
        color {colorUseRule=UseColor_Specified, specifiedColor=<rgba 0 0 0 0>}
        cornerRadii=<0 0 0 0>
       }
       tintColor {colorUseRule=UseColor_Specified, specifiedColor=<rgba 1 1 1 1>}
       uVRegion {bIsValid=false, max=<0 0>, min=<0 0>}
      }
      hoveredThumbImage {
       bIsDynamicallyLoaded=false
       drawAs=RoundedBox
       imageType=NoImage
       mirroring=NoMirror
       resourceName=None
       resourceObject=null
       tiling=NoTile
       imageSize=<0 0>
       margin=<ltrb 0 0 0 0>
       outlineSettings {
        bUseBrushTransparency=false
        roundingType=FixedRadius
        width=0
        color {colorUseRule=UseColor_Specified, specifiedColor=<rgba 0 0 0 0>}
        cornerRadii=<4 4 4 4>
       }
       tintColor {colorUseRule=UseColor_Specified, specifiedColor=<rgba 0.2159 0.2159 0.2159 1>}
       uVRegion {bIsValid=false, max=<0 0>, min=<0 0>}
      }
      normalThumbImage {
       bIsDynamicallyLoaded=false
       drawAs=RoundedBox
       imageType=NoImage
       mirroring=NoMirror
       resourceName=None
       resourceObject=null
       tiling=NoTile
       imageSize=<0 0>
       margin=<ltrb 0 0 0 0>
       outlineSettings {
        bUseBrushTransparency=false
        roundingType=FixedRadius
        width=0
        color {colorUseRule=UseColor_Specified, specifiedColor=<rgba 0 0 0 0>}
        cornerRadii=<4 4 4 4>
       }
       tintColor {colorUseRule=UseColor_Specified, specifiedColor=<rgba 0.0953 0.0953 0.0953 1>}
       uVRegion {bIsValid=false, max=<0 0>, min=<0 0>}
      }
      verticalBackgroundImage {
       bIsDynamicallyLoaded=false
       drawAs=NoDrawType
       imageType=NoImage
       mirroring=NoMirror
       resourceName=None
       resourceObject=null
       tiling=NoTile
       imageSize=<0 0>
       margin=<ltrb 0 0 0 0>
       outlineSettings {
        bUseBrushTransparency=false
        roundingType=HalfHeightRadius
        width=0
        color {colorUseRule=UseColor_Specified, specifiedColor=<rgba 0 0 0 0>}
        cornerRadii=<0 0 0 0>
       }
       tintColor {colorUseRule=UseColor_Specified, specifiedColor=<rgba 1 1 1 1>}
       uVRegion {bIsValid=false, max=<0 0>, min=<0 0>}
      }
      verticalBottomSlotImage {
       bIsDynamicallyLoaded=false
       drawAs=NoDrawType
       imageType=NoImage
       mirroring=NoMirror
       resourceName=None
       resourceObject=null
       tiling=NoTile
       imageSize=<0 0>
       margin=<ltrb 0 0 0 0>
       outlineSettings {
        bUseBrushTransparency=false
        roundingType=HalfHeightRadius
        width=0
        color {colorUseRule=UseColor_Specified, specifiedColor=<rgba 0 0 0 0>}
        cornerRadii=<0 0 0 0>
       }
       tintColor {colorUseRule=UseColor_Specified, specifiedColor=<rgba 1 1 1 1>}
       uVRegion {bIsValid=false, max=<0 0>, min=<0 0>}
      }
      verticalTopSlotImage {
       bIsDynamicallyLoaded=false
       drawAs=NoDrawType
       imageType=NoImage
       mirroring=NoMirror
       resourceName=None
       resourceObject=null
       tiling=NoTile
       imageSize=<0 0>
       margin=<ltrb 0 0 0 0>
       outlineSettings {
        bUseBrushTransparency=false
        roundingType=HalfHeightRadius
        width=0
        color {colorUseRule=UseColor_Specified, specifiedColor=<rgba 0 0 0 0>}
        cornerRadii=<0 0 0 0>
       }
       tintColor {colorUseRule=UseColor_Specified, specifiedColor=<rgba 1 1 1 1>}
       uVRegion {bIsValid=false, max=<0 0>, min=<0 0>}
      }
     }
     textStyle {
      overflowPolicy=Clip
      transformPolicy=None
      colorAndOpacity {colorUseRule=UseColor_Foreground, specifiedColor=<rgba 1 0 1 1>}
      font {
       bForceMonospaced=false
       bMaterialIsStencil=false
       fontMaterial=null
       fontObject=null
       letterSpacing=0
       monospacedWidth=1
       size=10
       skewAmount=0
       typefaceFontName=Regular
       outlineSettings {
        bApplyOutlineToDropShadows=false
        bMiteredCorners=false
        bSeparateFillAlpha=false
        outlineMaterial=null
        outlineSize=0
        outlineColor=<rgba 0 0 0 1>
       }
      }
      highlightColor {colorUseRule=UseColor_ColorTable, specifiedColor=<rgba 0 0 0 0>}
      highlightShape {
       bIsDynamicallyLoaded=false
       drawAs=Image
       imageType=NoImage
       mirroring=NoMirror
       resourceName=None
       resourceObject=null
       tiling=NoTile
       imageSize=<0 0>
       margin=<ltrb 0 0 0 0>
       outlineSettings {
        bUseBrushTransparency=false
        roundingType=HalfHeightRadius
        width=0
        color {colorUseRule=UseColor_Specified, specifiedColor=<rgba 0 0 0 0>}
        cornerRadii=<0 0 0 0>
       }
       tintColor {colorUseRule=UseColor_ColorTable, specifiedColor=<rgba 0 0 0 0>}
       uVRegion {bIsValid=false, max=<0 0>, min=<0 0>}
      }
      selectedBackgroundColor {colorUseRule=UseColor_ColorTable, specifiedColor=<rgba 0 0 0 0>}
      shadowColorAndOpacity=<rgba 0 0 0 1>
      shadowOffset=<0 0>
      strikeBrush {
       bIsDynamicallyLoaded=false
       drawAs=Image
       imageType=NoImage
       mirroring=NoMirror
       resourceName=None
       resourceObject=null
       tiling=NoTile
       imageSize=<32 32>
       margin=<ltrb 0 0 0 0>
       outlineSettings {
        bUseBrushTransparency=false
        roundingType=HalfHeightRadius
        width=0
        color {colorUseRule=UseColor_Specified, specifiedColor=<rgba 0 0 0 0>}
        cornerRadii=<0 0 0 0>
       }
       tintColor {colorUseRule=UseColor_Specified, specifiedColor=<rgba 1 1 1 1>}
       uVRegion {bIsValid=false, max=<0 0>, min=<0 0>}
      }
      underlineBrush {
       bIsDynamicallyLoaded=false
       drawAs=Image
       imageType=NoImage
       mirroring=NoMirror
       resourceName=None
       resourceObject=null
       tiling=NoTile
       imageSize=<32 32>
       margin=<ltrb 0 0 0 0>
       outlineSettings {
        bUseBrushTransparency=false
        roundingType=HalfHeightRadius
        width=0
        color {colorUseRule=UseColor_Specified, specifiedColor=<rgba 0 0 0 0>}
        cornerRadii=<0 0 0 0>
       }
       tintColor {colorUseRule=UseColor_Specified, specifiedColor=<rgba 1 1 1 1>}
       uVRegion {bIsValid=false, max=<0 0>, min=<0 0>}
      }
     }
     vScrollBarPadding=<ltrb 0 0 0 0>
    }
   }
   slot {class=HorizontalBoxSlot, properties {size {sizeRule=Fill, value=1}}}
  }
  BugItOutputLocation {
   widgetClassPath=&HorizontalBox
   parent=VerticalBox
   widget=BugItOutputLocation
   slot {class=VerticalBoxSlot, properties {padding=<ltrb 0 0 0 5>}}
  }
  Bugit_Output {
   widgetClassPath=&EditorUtilityEditableTextBox
   bIsVariable=true
   parent=BugItOutputLocation
   widget=Bugit_Output
   properties {
    hintText=BugIt File Location Output
    justification=Center
    keyboardType=Number
    selectAllTextWhenFocused=true
    widgetStyle {
     backgroundColor {colorUseRule=UseColor_Specified, specifiedColor=<rgba 1 1 1 1>}
     backgroundImageFocused {
      bIsDynamicallyLoaded=false
      drawAs=RoundedBox
      imageType=NoImage
      mirroring=NoMirror
      resourceName=None
      resourceObject=null
      tiling=NoTile
      imageSize=<0 0>
      margin=<ltrb 0 0 0 0>
      outlineSettings {
       bUseBrushTransparency=false
       roundingType=FixedRadius
       width=1
       color {colorUseRule=UseColor_ColorTable, specifiedColor=<rgba 0 0 0 0>}
       cornerRadii=<4 4 4 4>
      }
      tintColor {colorUseRule=UseColor_Specified, specifiedColor=<rgba 0.0048 0.0048 0.0048 1>}
      uVRegion {bIsValid=false, max=<0 0>, min=<0 0>}
     }
     backgroundImageHovered {
      bIsDynamicallyLoaded=false
      drawAs=RoundedBox
      imageType=NoImage
      mirroring=NoMirror
      resourceName=None
      resourceObject=null
      tiling=NoTile
      imageSize=<0 0>
      margin=<ltrb 0 0 0 0>
      outlineSettings {
       bUseBrushTransparency=false
       roundingType=FixedRadius
       width=1
       color {colorUseRule=UseColor_ColorTable, specifiedColor=<rgba 0 0 0 0>}
       cornerRadii=<4 4 4 4>
      }
      tintColor {colorUseRule=UseColor_Specified, specifiedColor=<rgba 0.0048 0.0048 0.0048 1>}
      uVRegion {bIsValid=false, max=<0 0>, min=<0 0>}
     }
     backgroundImageNormal {
      bIsDynamicallyLoaded=false
      drawAs=RoundedBox
      imageType=NoImage
      mirroring=NoMirror
      resourceName=None
      resourceObject=null
      tiling=NoTile
      imageSize=<0 0>
      margin=<ltrb 0 0 0 0>
      outlineSettings {
       bUseBrushTransparency=false
       roundingType=FixedRadius
       width=1
       color {colorUseRule=UseColor_ColorTable, specifiedColor=<rgba 0 0 0 0>}
       cornerRadii=<4 4 4 4>
      }
      tintColor {colorUseRule=UseColor_Specified, specifiedColor=<rgba 0.0048 0.0048 0.0048 1>}
      uVRegion {bIsValid=false, max=<0 0>, min=<0 0>}
     }
     backgroundImageReadOnly {
      bIsDynamicallyLoaded=false
      drawAs=RoundedBox
      imageType=NoImage
      mirroring=NoMirror
      resourceName=None
      resourceObject=null
      tiling=NoTile
      imageSize=<0 0>
      margin=<ltrb 0 0 0 0>
      outlineSettings {
       bUseBrushTransparency=false
       roundingType=FixedRadius
       width=1
       color {colorUseRule=UseColor_ColorTable, specifiedColor=<rgba 0 0 0 0>}
       cornerRadii=<4 4 4 4>
      }
      tintColor {colorUseRule=UseColor_Specified, specifiedColor=<rgba 0.0284 0.0284 0.0284 1>}
      uVRegion {bIsValid=false, max=<0 0>, min=<0 0>}
     }
     focusedForegroundColor {colorUseRule=UseColor_Specified, specifiedColor=<rgba 1 1 1 1>}
     foregroundColor {colorUseRule=UseColor_Specified, specifiedColor=<rgba 0.5271 0.5271 0.5271 1>}
     hScrollBarPadding=<ltrb 0 0 0 0>
     padding=<ltrb 12 3 8 4>
     readOnlyForegroundColor {colorUseRule=UseColor_Foreground, specifiedColor=<rgba 1 0 1 1>}
     scrollBarStyle {
      thickness=8
      draggedThumbImage {
       bIsDynamicallyLoaded=false
       drawAs=RoundedBox
       imageType=NoImage
       mirroring=NoMirror
       resourceName=None
       resourceObject=null
       tiling=NoTile
       imageSize=<0 0>
       margin=<ltrb 0 0 0 0>
       outlineSettings {
        bUseBrushTransparency=false
        roundingType=FixedRadius
        width=0
        color {colorUseRule=UseColor_Specified, specifiedColor=<rgba 0 0 0 0>}
        cornerRadii=<4 4 4 4>
       }
       tintColor {colorUseRule=UseColor_Specified, specifiedColor=<rgba 0.2159 0.2159 0.2159 1>}
       uVRegion {bIsValid=false, max=<0 0>, min=<0 0>}
      }
      horizontalBackgroundImage {
       bIsDynamicallyLoaded=false
       drawAs=NoDrawType
       imageType=NoImage
       mirroring=NoMirror
       resourceName=None
       resourceObject=null
       tiling=NoTile
       imageSize=<0 0>
       margin=<ltrb 0 0 0 0>
       outlineSettings {
        bUseBrushTransparency=false
        roundingType=HalfHeightRadius
        width=0
        color {colorUseRule=UseColor_Specified, specifiedColor=<rgba 0 0 0 0>}
        cornerRadii=<0 0 0 0>
       }
       tintColor {colorUseRule=UseColor_Specified, specifiedColor=<rgba 1 1 1 1>}
       uVRegion {bIsValid=false, max=<0 0>, min=<0 0>}
      }
      horizontalBottomSlotImage {
       bIsDynamicallyLoaded=false
       drawAs=NoDrawType
       imageType=NoImage
       mirroring=NoMirror
       resourceName=None
       resourceObject=null
       tiling=NoTile
       imageSize=<0 0>
       margin=<ltrb 0 0 0 0>
       outlineSettings {
        bUseBrushTransparency=false
        roundingType=HalfHeightRadius
        width=0
        color {colorUseRule=UseColor_Specified, specifiedColor=<rgba 0 0 0 0>}
        cornerRadii=<0 0 0 0>
       }
       tintColor {colorUseRule=UseColor_Specified, specifiedColor=<rgba 1 1 1 1>}
       uVRegion {bIsValid=false, max=<0 0>, min=<0 0>}
      }
      horizontalTopSlotImage {
       bIsDynamicallyLoaded=false
       drawAs=NoDrawType
       imageType=NoImage
       mirroring=NoMirror
       resourceName=None
       resourceObject=null
       tiling=NoTile
       imageSize=<0 0>
       margin=<ltrb 0 0 0 0>
       outlineSettings {
        bUseBrushTransparency=false
        roundingType=HalfHeightRadius
        width=0
        color {colorUseRule=UseColor_Specified, specifiedColor=<rgba 0 0 0 0>}
        cornerRadii=<0 0 0 0>
       }
       tintColor {colorUseRule=UseColor_Specified, specifiedColor=<rgba 1 1 1 1>}
       uVRegion {bIsValid=false, max=<0 0>, min=<0 0>}
      }
      hoveredThumbImage {
       bIsDynamicallyLoaded=false
       drawAs=RoundedBox
       imageType=NoImage
       mirroring=NoMirror
       resourceName=None
       resourceObject=null
       tiling=NoTile
       imageSize=<0 0>
       margin=<ltrb 0 0 0 0>
       outlineSettings {
        bUseBrushTransparency=false
        roundingType=FixedRadius
        width=0
        color {colorUseRule=UseColor_Specified, specifiedColor=<rgba 0 0 0 0>}
        cornerRadii=<4 4 4 4>
       }
       tintColor {colorUseRule=UseColor_Specified, specifiedColor=<rgba 0.2159 0.2159 0.2159 1>}
       uVRegion {bIsValid=false, max=<0 0>, min=<0 0>}
      }
      normalThumbImage {
       bIsDynamicallyLoaded=false
       drawAs=RoundedBox
       imageType=NoImage
       mirroring=NoMirror
       resourceName=None
       resourceObject=null
       tiling=NoTile
       imageSize=<0 0>
       margin=<ltrb 0 0 0 0>
       outlineSettings {
        bUseBrushTransparency=false
        roundingType=FixedRadius
        width=0
        color {colorUseRule=UseColor_Specified, specifiedColor=<rgba 0 0 0 0>}
        cornerRadii=<4 4 4 4>
       }
       tintColor {colorUseRule=UseColor_Specified, specifiedColor=<rgba 0.0953 0.0953 0.0953 1>}
       uVRegion {bIsValid=false, max=<0 0>, min=<0 0>}
      }
      verticalBackgroundImage {
       bIsDynamicallyLoaded=false
       drawAs=NoDrawType
       imageType=NoImage
       mirroring=NoMirror
       resourceName=None
       resourceObject=null
       tiling=NoTile
       imageSize=<0 0>
       margin=<ltrb 0 0 0 0>
       outlineSettings {
        bUseBrushTransparency=false
        roundingType=HalfHeightRadius
        width=0
        color {colorUseRule=UseColor_Specified, specifiedColor=<rgba 0 0 0 0>}
        cornerRadii=<0 0 0 0>
       }
       tintColor {colorUseRule=UseColor_Specified, specifiedColor=<rgba 1 1 1 1>}
       uVRegion {bIsValid=false, max=<0 0>, min=<0 0>}
      }
      verticalBottomSlotImage {
       bIsDynamicallyLoaded=false
       drawAs=NoDrawType
       imageType=NoImage
       mirroring=NoMirror
       resourceName=None
       resourceObject=null
       tiling=NoTile
       imageSize=<0 0>
       margin=<ltrb 0 0 0 0>
       outlineSettings {
        bUseBrushTransparency=false
        roundingType=HalfHeightRadius
        width=0
        color {colorUseRule=UseColor_Specified, specifiedColor=<rgba 0 0 0 0>}
        cornerRadii=<0 0 0 0>
       }
       tintColor {colorUseRule=UseColor_Specified, specifiedColor=<rgba 1 1 1 1>}
       uVRegion {bIsValid=false, max=<0 0>, min=<0 0>}
      }
      verticalTopSlotImage {
       bIsDynamicallyLoaded=false
       drawAs=NoDrawType
       imageType=NoImage
       mirroring=NoMirror
       resourceName=None
       resourceObject=null
       tiling=NoTile
       imageSize=<0 0>
       margin=<ltrb 0 0 0 0>
       outlineSettings {
        bUseBrushTransparency=false
        roundingType=HalfHeightRadius
        width=0
        color {colorUseRule=UseColor_Specified, specifiedColor=<rgba 0 0 0 0>}
        cornerRadii=<0 0 0 0>
       }
       tintColor {colorUseRule=UseColor_Specified, specifiedColor=<rgba 1 1 1 1>}
       uVRegion {bIsValid=false, max=<0 0>, min=<0 0>}
      }
     }
     textStyle {
      overflowPolicy=Clip
      transformPolicy=None
      colorAndOpacity {colorUseRule=UseColor_Foreground, specifiedColor=<rgba 1 0 1 1>}
      font {
       bForceMonospaced=false
       bMaterialIsStencil=false
       fontMaterial=null
       fontObject=null
       letterSpacing=0
       monospacedWidth=1
       size=10
       skewAmount=0
       typefaceFontName=Regular
       outlineSettings {
        bApplyOutlineToDropShadows=false
        bMiteredCorners=false
        bSeparateFillAlpha=false
        outlineMaterial=null
        outlineSize=0
        outlineColor=<rgba 0 0 0 1>
       }
      }
      highlightColor {colorUseRule=UseColor_ColorTable, specifiedColor=<rgba 0 0 0 0>}
      highlightShape {
       bIsDynamicallyLoaded=false
       drawAs=Image
       imageType=NoImage
       mirroring=NoMirror
       resourceName=None
       resourceObject=null
       tiling=NoTile
       imageSize=<0 0>
       margin=<ltrb 0 0 0 0>
       outlineSettings {
        bUseBrushTransparency=false
        roundingType=HalfHeightRadius
        width=0
        color {colorUseRule=UseColor_Specified, specifiedColor=<rgba 0 0 0 0>}
        cornerRadii=<0 0 0 0>
       }
       tintColor {colorUseRule=UseColor_ColorTable, specifiedColor=<rgba 0 0 0 0>}
       uVRegion {bIsValid=false, max=<0 0>, min=<0 0>}
      }
      selectedBackgroundColor {colorUseRule=UseColor_ColorTable, specifiedColor=<rgba 0 0 0 0>}
      shadowColorAndOpacity=<rgba 0 0 0 1>
      shadowOffset=<0 0>
      strikeBrush {
       bIsDynamicallyLoaded=false
       drawAs=Image
       imageType=NoImage
       mirroring=NoMirror
       resourceName=None
       resourceObject=null
       tiling=NoTile
       imageSize=<32 32>
       margin=<ltrb 0 0 0 0>
       outlineSettings {
        bUseBrushTransparency=false
        roundingType=HalfHeightRadius
        width=0
        color {colorUseRule=UseColor_Specified, specifiedColor=<rgba 0 0 0 0>}
        cornerRadii=<0 0 0 0>
       }
       tintColor {colorUseRule=UseColor_Specified, specifiedColor=<rgba 1 1 1 1>}
       uVRegion {bIsValid=false, max=<0 0>, min=<0 0>}
      }
      underlineBrush {
       bIsDynamicallyLoaded=false
       drawAs=Image
       imageType=NoImage
       mirroring=NoMirror
       resourceName=None
       resourceObject=null
       tiling=NoTile
       imageSize=<32 32>
       margin=<ltrb 0 0 0 0>
       outlineSettings {
        bUseBrushTransparency=false
        roundingType=HalfHeightRadius
        width=0
        color {colorUseRule=UseColor_Specified, specifiedColor=<rgba 0 0 0 0>}
        cornerRadii=<0 0 0 0>
       }
       tintColor {colorUseRule=UseColor_Specified, specifiedColor=<rgba 1 1 1 1>}
       uVRegion {bIsValid=false, max=<0 0>, min=<0 0>}
      }
     }
     vScrollBarPadding=<ltrb 0 0 0 0>
    }
   }
   slot {class=HorizontalBoxSlot, properties {size {sizeRule=Fill, value=1}}}
  }
  Separator {widgetClassPath=&HorizontalBox, parent=GeneralCheats_Box, widget=Separator, slot {class=VerticalBoxSlot}}
  Border_127 {
   widgetClassPath=&Border
   parent=Separator
   widget=Border_127
   slot {class=HorizontalBoxSlot, properties {size {sizeRule=Fill, value=1}}}
   properties=~
  }
  EventCheats {widgetClassPath=&VerticalBox, parent=EditorUtilityScrollBox_120, widget=EventCheats, slot {class=ScrollBoxSlot}}
  EventsTitle {
   widgetClassPath=&TextBlock
   parent=EventCheats
   widget=EventsTitle
   properties {
    text=Events
    font {
     bForceMonospaced=false
     bMaterialIsStencil=false
     fontMaterial=null
     fontObject=&Roboto
     letterSpacing=0
     monospacedWidth=1
     size=18
     skewAmount=0
     typefaceFontName=Bold
     outlineSettings {
      bApplyOutlineToDropShadows=false
      bMiteredCorners=false
      bSeparateFillAlpha=false
      outlineMaterial=null
      outlineSize=0
      outlineColor=<rgba 0 0 0 1>
     }
    }
   }
   slot {class=VerticalBoxSlot, properties {horizontalAlignment=HAlign_Center}}
  }
  EventsCheats_Box {widgetClassPath=&VerticalBox, parent=EventCheats, widget=EventsCheats_Box, slot {class=VerticalBoxSlot}}
  PlayEvent {
   widgetClassPath=&HorizontalBox
   parent=EventsCheats_Box
   widget=PlayEvent
   slot {class=VerticalBoxSlot, properties {padding=<ltrb 0 0 0 5>}}
  }
  Events_PlayEventButton {
   widgetClassPath=&EditorUtilityButton
   parent=PlayEvent
   widget=Events_PlayEventButton
   properties {toolTipText="Uses cheat \"PlayEvent\" with the numbers entered in the next two input boxes."}
   slot {class=HorizontalBoxSlot, properties {padding=<ltrb 2.5 0 0 0>, size {sizeRule=Fill, value=1}}}
  }
  TextBlock_364 {
   widgetClassPath=&TextBlock
   parent=Events_PlayEventButton
   widget=TextBlock_364
   properties {
    text=Play Event
    font {
     bForceMonospaced=false
     bMaterialIsStencil=false
     fontMaterial=null
     fontObject=&Roboto
     letterSpacing=0
     monospacedWidth=1
     size=9
     skewAmount=0
     typefaceFontName=Regular
     outlineSettings {
      bApplyOutlineToDropShadows=false
      bMiteredCorners=false
      bSeparateFillAlpha=false
      outlineMaterial=null
      outlineSize=0
      outlineColor=<rgba 0 0 0 1>
     }
    }
   }
   slot {class=ButtonSlot}
  }
  Events_EventNumber {
   widgetClassPath=&EditorUtilityEditableTextBox
   bIsVariable=true
   parent=PlayEvent
   widget=Events_EventNumber
   properties {hintText=Event Index, justification=Center, keyboardType=Number, selectAllTextWhenFocused=true}
   slot {class=HorizontalBoxSlot, properties {size {sizeRule=Fill, value=1}}}
  }
  Events_Second {
   widgetClassPath=&EditorUtilityEditableTextBox
   bIsVariable=true
   parent=PlayEvent
   widget=Events_Second
   properties {hintText=Seconds, justification=Center, keyboardType=Number, selectAllTextWhenFocused=true}
   slot {class=HorizontalBoxSlot, properties {padding=<ltrb 0 0 2.5 0>, size {sizeRule=Fill, value=1}}}
  }
  EventSequenceOptions {
   widgetClassPath=&HorizontalBox
   parent=EventsCheats_Box
   widget=EventSequenceOptions
   slot {class=VerticalBoxSlot, properties {padding=<ltrb 0 0 0 5>}}
  }
  Events_Play {
   widgetClassPath=&EditorUtilityButton
   parent=EventSequenceOptions
   widget=Events_Play
   properties {toolTipText="Freeze the game in its current state. ALERT: If you do this, it will be frozen for good. You have to restart PIE to continue."}
   slot {class=HorizontalBoxSlot, properties {padding=<ltrb 2.5 0 2.5 0>, size {sizeRule=Fill, value=1}}}
  }
  PlayPause {
   widgetClassPath=&TextBlock
   parent=Events_Play
   widget=PlayPause
   properties {
    text=Freeze
    font {
     bForceMonospaced=false
     bMaterialIsStencil=false
     fontMaterial=null
     fontObject=&Roboto
     letterSpacing=0
     monospacedWidth=1
     size=9
     skewAmount=0
     typefaceFontName=Regular
     outlineSettings {
      bApplyOutlineToDropShadows=false
      bMiteredCorners=false
      bSeparateFillAlpha=false
      outlineMaterial=null
      outlineSize=0
      outlineColor=<rgba 0 0 0 1>
     }
    }
   }
   slot {class=ButtonSlot}
  }
  Separator_1 {widgetClassPath=&HorizontalBox, parent=EventCheats, widget=Separator_1, slot {class=VerticalBoxSlot}}
  Border {
   widgetClassPath=&Border
   parent=Separator_1
   widget=Border
   slot {class=HorizontalBoxSlot, properties {size {sizeRule=Fill, value=1}}}
   properties=~
  }
  Utilities {widgetClassPath=&VerticalBox, parent=EditorUtilityScrollBox_120, widget=Utilities, slot {class=ScrollBoxSlot}}
  UtilityTitle {
   widgetClassPath=&TextBlock
   parent=Utilities
   widget=UtilityTitle
   properties {
    text=Utilities
    font {
     bForceMonospaced=false
     bMaterialIsStencil=false
     fontMaterial=null
     fontObject=&Roboto
     letterSpacing=0
     monospacedWidth=1
     size=18
     skewAmount=0
     typefaceFontName=Bold
     outlineSettings {
      bApplyOutlineToDropShadows=false
      bMiteredCorners=false
      bSeparateFillAlpha=false
      outlineMaterial=null
      outlineSize=0
      outlineColor=<rgba 0 0 0 1>
     }
    }
   }
   slot {class=VerticalBoxSlot, properties {horizontalAlignment=HAlign_Center}}
  }
  UtilityCheats_Box {widgetClassPath=&HorizontalBox, parent=Utilities, widget=UtilityCheats_Box, slot {class=VerticalBoxSlot}}
  Util_LeftColumn {
   widgetClassPath=&VerticalBox
   parent=UtilityCheats_Box
   widget=Util_LeftColumn
   slot {class=HorizontalBoxSlot, properties {size {sizeRule=Fill, value=1}}}
  }
  ScreenMessages {
   widgetClassPath=&HorizontalBox
   parent=Util_LeftColumn
   widget=ScreenMessages
   properties {toolTipText=This will stop those pesky messages on the top left of your screen.}
   slot {class=VerticalBoxSlot, properties {padding=<ltrb 0 0 0 5>, size {sizeRule=Fill, value=1}}}
  }
  TextBlock_4 {
   widgetClassPath=&TextBlock
   parent=ScreenMessages
   widget=TextBlock_4
   properties {
    text=Print Screen Messages
    font {
     bForceMonospaced=false
     bMaterialIsStencil=false
     fontMaterial=null
     fontObject=&Roboto
     letterSpacing=0
     monospacedWidth=1
     size=9
     skewAmount=0
     typefaceFontName=Regular
     outlineSettings {
      bApplyOutlineToDropShadows=false
      bMiteredCorners=false
      bSeparateFillAlpha=false
      outlineMaterial=null
      outlineSize=0
      outlineColor=<rgba 0 0 0 1>
     }
    }
   }
   slot {class=HorizontalBoxSlot, properties {verticalAlignment=VAlign_Center, padding=<ltrb 10 0 0 0>}}
  }
  CheckBox_ScreenMessages {
   widgetClassPath=&EditorUtilityCheckBox
   bIsVariable=true
   parent=ScreenMessages
   widget=CheckBox_ScreenMessages
   properties {checkedState=Checked}
   slot {class=HorizontalBoxSlot, properties {horizontalAlignment=HAlign_Right, padding=<ltrb 0 0 5 0>, size {sizeRule=Fill, value=1}}}
  }
  TextBlock_8 {
   widgetClassPath=&TextBlock
   parent=Util_LeftColumn
   widget=TextBlock_8
   properties {
    justification=Center
    text="Gameplay Speed (Slow Motion)"
    font {
     bForceMonospaced=false
     bMaterialIsStencil=false
     fontMaterial=null
     fontObject=&Roboto
     letterSpacing=0
     monospacedWidth=1
     size=9
     skewAmount=0
     typefaceFontName=Regular
     outlineSettings {
      bApplyOutlineToDropShadows=false
      bMiteredCorners=false
      bSeparateFillAlpha=false
      outlineMaterial=null
      outlineSize=0
      outlineColor=<rgba 0 0 0 1>
     }
    }
   }
   slot {class=VerticalBoxSlot, properties {horizontalAlignment=HAlign_Center, verticalAlignment=VAlign_Center, padding=<ltrb 0 0 0 2.5>}}
  }
  Utility_Slomo_Slider {
   widgetClassPath=&EditorUtilitySpinBox
   bIsVariable=true
   parent=Utilities
   widget=Utility_Slomo_Slider
   properties {
    bOverride_MaxValue=true
    bOverride_MinValue=true
    maxValue=1
    toolTipText=Use this slider to change the slow motion intensity.
    value=1
    foregroundColor {colorUseRule=UseColor_Specified, specifiedColor=<rgba 1 1 1 1>}
   }
   slot {class=VerticalBoxSlot, properties {padding=<ltrb 2.5 0 2.5 5>, size {sizeRule=Fill, value=1}}}
  }
  Viewmode {
   widgetClassPath=&HorizontalBox
   parent=Utilities
   widget=Viewmode
   properties {toolTipText=Use this dropdown to change your viewmode quickly during PIE}
   slot {class=VerticalBoxSlot, properties {padding=<ltrb 0 0 0 5>}}
  }
  TextBlock_6 {
   widgetClassPath=&TextBlock
   parent=Viewmode
   widget=TextBlock_6
   properties {
    text=Viewmode
    font {
     bForceMonospaced=false
     bMaterialIsStencil=false
     fontMaterial=null
     fontObject=&Roboto
     letterSpacing=0
     monospacedWidth=1
     size=9
     skewAmount=0
     typefaceFontName=Regular
     outlineSettings {
      bApplyOutlineToDropShadows=false
      bMiteredCorners=false
      bSeparateFillAlpha=false
      outlineMaterial=null
      outlineSize=0
      outlineColor=<rgba 0 0 0 1>
     }
    }
   }
   slot {
    class=HorizontalBoxSlot
    properties {horizontalAlignment=HAlign_Center, verticalAlignment=VAlign_Center, padding=<ltrb 4 2 4 2>, size {sizeRule=Fill, value=0.5}}
   }
  }
  ComboBox_Viewmode {
   widgetClassPath=/Script/Blutility.EditorUtilityComboBoxString
   bIsVariable=true
   parent=Viewmode
   widget=ComboBox_Viewmode
   properties {
    selectedOption=Lit
    defaultOptions [Lit,Wireframe,Unlit]
    font {
     bForceMonospaced=false
     bMaterialIsStencil=false
     fontMaterial=null
     fontObject=&Roboto
     letterSpacing=0
     monospacedWidth=1
     size=9
     skewAmount=0
     typefaceFontName=Regular
     outlineSettings {
      bApplyOutlineToDropShadows=false
      bMiteredCorners=false
      bSeparateFillAlpha=false
      outlineMaterial=null
      outlineSize=0
      outlineColor=<rgba 0 0 0 1>
     }
    }
    foregroundColor {colorUseRule=UseColor_Specified, specifiedColor=<rgba 0.5271 0.5271 0.5271 1>}
   }
   slot {class=HorizontalBoxSlot, properties {padding=<ltrb 0 0 2.5 0>, size {sizeRule=Fill, value=1}}}
  }
  DebugPhysics {widgetClassPath=&HorizontalBox, parent=Utilities, widget=DebugPhysics, slot {class=VerticalBoxSlot}}
  Button_DebugPhysics {
   widgetClassPath=&EditorUtilityButton
   parent=DebugPhysics
   widget=Button_DebugPhysics
   properties {toolTipText=Selecting this will begin drawing actor collisions in the world.}
   slot {class=HorizontalBoxSlot, properties {padding=<ltrb 2.5 0 2.5 5>, size {sizeRule=Fill, value=1}}}
  }
  TextBlock_7 {
   widgetClassPath=&TextBlock
   parent=Button_DebugPhysics
   widget=TextBlock_7
   properties {
    text=Show Debug Physics
    font {
     bForceMonospaced=false
     bMaterialIsStencil=false
     fontMaterial=null
     fontObject=&Roboto
     letterSpacing=0
     monospacedWidth=1
     size=9
     skewAmount=0
     typefaceFontName=Regular
     outlineSettings {
      bApplyOutlineToDropShadows=false
      bMiteredCorners=false
      bSeparateFillAlpha=false
      outlineMaterial=null
      outlineSize=0
      outlineColor=<rgba 0 0 0 1>
     }
    }
   }
   slot {class=ButtonSlot}
  }
  Separator_2 {widgetClassPath=&HorizontalBox, parent=Utilities, widget=Separator_2, slot {class=VerticalBoxSlot}}
  Border_1 {
   widgetClassPath=&Border
   parent=Separator_2
   widget=Border_1
   slot {class=HorizontalBoxSlot, properties {size {sizeRule=Fill, value=1}}}
   properties=~
  }
  Spawn {widgetClassPath=&VerticalBox, parent=EditorUtilityScrollBox_120, widget=Spawn, slot {class=ScrollBoxSlot}}
  Spawn_Title {
   widgetClassPath=&TextBlock
   parent=Spawn
   widget=Spawn_Title
   properties {
    text=Spawn
    font {
     bForceMonospaced=false
     bMaterialIsStencil=false
     fontMaterial=null
     fontObject=&Roboto
     letterSpacing=0
     monospacedWidth=1
     size=18
     skewAmount=0
     typefaceFontName=Bold
     outlineSettings {
      bApplyOutlineToDropShadows=false
      bMiteredCorners=false
      bSeparateFillAlpha=false
      outlineMaterial=null
      outlineSize=0
      outlineColor=<rgba 0 0 0 1>
     }
    }
   }
   slot {class=VerticalBoxSlot, properties {horizontalAlignment=HAlign_Center}}
  }
  VehicleCheats {
   widgetClassPath=&HorizontalBox
   parent=Spawn
   widget=VehicleCheats
   slot {class=VerticalBoxSlot, properties {padding=<ltrb 0 0 0 2.5>}}
  }
  Spawn_VehicleExpand {
   widgetClassPath=&EditorUtilityExpandableArea
   bIsVariable=true
   parent=VehicleCheats
   widget=Spawn_VehicleExpand
   slot {class=HorizontalBoxSlot, properties {size {sizeRule=Fill, value=1}}}
  }
  TextBlock_521 {
   widgetClassPath=&TextBlock
   namedSlotHost=Spawn_VehicleExpand
   widget=TextBlock_521
   properties {
    text=Vehicles
    toolTipText=A lot of spawn vehicle cheats.
    font {
     bForceMonospaced=false
     bMaterialIsStencil=false
     fontMaterial=null
     fontObject=&Roboto
     letterSpacing=0
     monospacedWidth=1
     size=9
     skewAmount=0
     typefaceFontName=Regular
     outlineSettings {
      bApplyOutlineToDropShadows=false
      bMiteredCorners=false
      bSeparateFillAlpha=false
      outlineMaterial=null
      outlineSize=0
      outlineColor=<rgba 0 0 0 1>
     }
    }
   }
  }
  VerticalBox_864 {widgetClassPath=&VerticalBox, namedSlotHost=Spawn_VehicleExpand, widget=VerticalBox_864}
  SingleProperty_Vehicle {
   widgetClassPath=&SinglePropertyView
   parent=VerticalBox_864
   widget=SingleProperty_Vehicle
   properties {nameOverride=Vehicle, propertyName=Property_Vehicle, toolTipText=Select the ID for the Vehicle you would like to spawn.}
   slot {class=VerticalBoxSlot}
  }
  Button_SpawnVehicle {
   widgetClassPath=&EditorUtilityButton
   parent=VerticalBox_864
   widget=Button_SpawnVehicle
   properties {toolTipText=Spawn the Vehicle using the VID selected above.}
   slot {class=VerticalBoxSlot, properties {padding=<ltrb 2.5 0 2.5 5>}}
  }
  TextBlock_1 {
   widgetClassPath=&TextBlock
   parent=Button_SpawnVehicle
   widget=TextBlock_1
   properties {
    text=Spawn Vehicle
    font {
     bForceMonospaced=false
     bMaterialIsStencil=false
     fontMaterial=null
     fontObject=&Roboto
     letterSpacing=0
     monospacedWidth=1
     size=9
     skewAmount=0
     typefaceFontName=Regular
     outlineSettings {
      bApplyOutlineToDropShadows=false
      bMiteredCorners=false
      bSeparateFillAlpha=false
      outlineMaterial=null
      outlineSize=0
      outlineColor=<rgba 0 0 0 1>
     }
    }
   }
   slot {class=ButtonSlot}
  }
  ItemCheats {
   widgetClassPath=&HorizontalBox
   parent=Spawn
   widget=ItemCheats
   slot {class=VerticalBoxSlot, properties {padding=<ltrb 0 0 0 2.5>}}
  }
  Spawn_ItemsExpand {
   widgetClassPath=&EditorUtilityExpandableArea
   bIsVariable=true
   parent=ItemCheats
   widget=Spawn_ItemsExpand
   slot {class=HorizontalBoxSlot, properties {size {sizeRule=Fill, value=1}}}
  }
  TextBlock_9 {
   widgetClassPath=&TextBlock
   namedSlotHost=Spawn_ItemsExpand
   widget=TextBlock_9
   properties {
    text=Items/Weapons
    toolTipText="Weapon, items, and consumable cheats."
    font {
     bForceMonospaced=false
     bMaterialIsStencil=false
     fontMaterial=null
     fontObject=&Roboto
     letterSpacing=0
     monospacedWidth=1
     size=9
     skewAmount=0
     typefaceFontName=Regular
     outlineSettings {
      bApplyOutlineToDropShadows=false
      bMiteredCorners=false
      bSeparateFillAlpha=false
      outlineMaterial=null
      outlineSize=0
      outlineColor=<rgba 0 0 0 1>
     }
    }
   }
  }
  VerticalBox_857 {widgetClassPath=&VerticalBox, namedSlotHost=Spawn_ItemsExpand, widget=VerticalBox_857}
  WeaponSingleProperty {
   widgetClassPath=&SinglePropertyView
   parent=VerticalBox_857
   widget=WeaponSingleProperty
   properties {nameOverride=Weapon, propertyName=SingleProperty_Weapon, toolTipText="Select an Item (Item Definition) to use for the commands."}
   slot {class=VerticalBoxSlot}
  }
  HorizontalBox_206 {widgetClassPath=&HorizontalBox, parent=VerticalBox_857, widget=HorizontalBox_206, slot {class=VerticalBoxSlot}}
  Spawn_WeaponPicker {
   widgetClassPath=&EditorUtilityButton
   parent=HorizontalBox_206
   widget=Spawn_WeaponPicker
   properties {toolTipText=Grant Item/Weapon to all players in the current PIE World.}
   slot {class=HorizontalBoxSlot, properties {padding=<ltrb 2.5 0 1.25 5>, size {sizeRule=Fill, value=1}}}
  }
  TextBlock {
   widgetClassPath=&TextBlock
   parent=Spawn_WeaponPicker
   widget=TextBlock
   properties {
    justification=Center
    text=Grant Item to Players
    font {
     bForceMonospaced=false
     bMaterialIsStencil=false
     fontMaterial=null
     fontObject=&Roboto
     letterSpacing=0
     monospacedWidth=1
     size=9
     skewAmount=0
     typefaceFontName=Regular
     outlineSettings {
      bApplyOutlineToDropShadows=false
      bMiteredCorners=false
      bSeparateFillAlpha=false
      outlineMaterial=null
      outlineSize=0
      outlineColor=<rgba 0 0 0 1>
     }
    }
   }
   slot {class=ButtonSlot, properties {horizontalAlignment=HAlign_Fill, verticalAlignment=VAlign_Fill}}
  }
  Drop_WeaponPicker {
   widgetClassPath=&EditorUtilityButton
   parent=HorizontalBox_206
   widget=Drop_WeaponPicker
   properties {toolTipText=Grant Item/Weapon to all players in the current PIE World.}
   slot {class=HorizontalBoxSlot, properties {padding=<ltrb 1.25 0 2.5 5>, size {sizeRule=Fill, value=1}}}
  }
  TextBlock_2 {
   widgetClassPath=&TextBlock
   parent=Drop_WeaponPicker
   widget=TextBlock_2
   properties {
    justification=Center
    text=Create Pickup
    font {
     bForceMonospaced=false
     bMaterialIsStencil=false
     fontMaterial=null
     fontObject=&Roboto
     letterSpacing=0
     monospacedWidth=1
     size=9
     skewAmount=0
     typefaceFontName=Regular
     outlineSettings {
      bApplyOutlineToDropShadows=false
      bMiteredCorners=false
      bSeparateFillAlpha=false
      outlineMaterial=null
      outlineSize=0
      outlineColor=<rgba 0 0 0 1>
     }
    }
   }
   slot {class=ButtonSlot, properties {horizontalAlignment=HAlign_Fill, verticalAlignment=VAlign_Fill}}
  }
  HorizontalBox_251 {widgetClassPath=&HorizontalBox, parent=VerticalBox_857, widget=HorizontalBox_251, slot {class=VerticalBoxSlot}}
  VerticalBox_270 {
   widgetClassPath=&VerticalBox
   parent=HorizontalBox_251
   widget=VerticalBox_270
   slot {class=HorizontalBoxSlot, properties {padding=<ltrb 0 0 1.25 0>, size {sizeRule=Fill, value=1}}}
  }
  TextBlock_46 {
   widgetClassPath=&TextBlock
   parent=VerticalBox_270
   widget=TextBlock_46
   properties {
    text=Weapons
    font {
     bForceMonospaced=false
     bMaterialIsStencil=false
     fontMaterial=null
     fontObject=&Roboto
     letterSpacing=0
     monospacedWidth=1
     size=9
     skewAmount=0
     typefaceFontName=Bold
     outlineSettings {
      bApplyOutlineToDropShadows=false
      bMiteredCorners=false
      bSeparateFillAlpha=false
      outlineMaterial=null
      outlineSize=0
      outlineColor=<rgba 0 0 0 1>
     }
    }
   }
   slot {
    class=VerticalBoxSlot
    properties {horizontalAlignment=HAlign_Center, verticalAlignment=VAlign_Top, padding=<ltrb 4 2 4 2>, size {sizeRule=Fill, value=1}}
   }
  }
  Spawn_GoodWeapons {widgetClassPath=&EditorUtilityButton, parent=VerticalBox_270, widget=Spawn_GoodWeapons, slot {class=VerticalBoxSlot}}
  TextBlock_49 {
   widgetClassPath=&TextBlock
   parent=Spawn_GoodWeapons
   widget=TextBlock_49
   properties {
    justification=Center
    text=Good Weapons
    font {
     bForceMonospaced=false
     bMaterialIsStencil=false
     fontMaterial=null
     fontObject=&Roboto
     letterSpacing=0
     monospacedWidth=1
     size=9
     skewAmount=0
     typefaceFontName=Regular
     outlineSettings {
      bApplyOutlineToDropShadows=false
      bMiteredCorners=false
      bSeparateFillAlpha=false
      outlineMaterial=null
      outlineSize=0
      outlineColor=<rgba 0 0 0 1>
     }
    }
   }
   slot {class=ButtonSlot, properties {horizontalAlignment=HAlign_Fill, verticalAlignment=VAlign_Fill}}
  }
  Spawn_AllWeapons {widgetClassPath=&EditorUtilityButton, parent=VerticalBox_270, widget=Spawn_AllWeapons, slot {class=VerticalBoxSlot}}
  TextBlock_50 {
   widgetClassPath=&TextBlock
   parent=Spawn_AllWeapons
   widget=TextBlock_50
   properties {
    justification=Center
    text=All Weapons
    font {
     bForceMonospaced=false
     bMaterialIsStencil=false
     fontMaterial=null
     fontObject=&Roboto
     letterSpacing=0
     monospacedWidth=1
     size=9
     skewAmount=0
     typefaceFontName=Regular
     outlineSettings {
      bApplyOutlineToDropShadows=false
      bMiteredCorners=false
      bSeparateFillAlpha=false
      outlineMaterial=null
      outlineSize=0
      outlineColor=<rgba 0 0 0 1>
     }
    }
   }
   slot {class=ButtonSlot, properties {horizontalAlignment=HAlign_Fill, verticalAlignment=VAlign_Fill}}
  }
  Spawn_Melee {widgetClassPath=&EditorUtilityButton, parent=VerticalBox_270, widget=Spawn_Melee, slot {class=VerticalBoxSlot}}
  TextBlock_51 {
   widgetClassPath=&TextBlock
   parent=Spawn_Melee
   widget=TextBlock_51
   properties {
    justification=Center
    text=All Melee
    font {
     bForceMonospaced=false
     bMaterialIsStencil=false
     fontMaterial=null
     fontObject=&Roboto
     letterSpacing=0
     monospacedWidth=1
     size=9
     skewAmount=0
     typefaceFontName=Regular
     outlineSettings {
      bApplyOutlineToDropShadows=false
      bMiteredCorners=false
      bSeparateFillAlpha=false
      outlineMaterial=null
      outlineSize=0
      outlineColor=<rgba 0 0 0 1>
     }
    }
   }
   slot {class=ButtonSlot, properties {horizontalAlignment=HAlign_Fill, verticalAlignment=VAlign_Fill}}
  }
  Spawn_Pistols {widgetClassPath=&EditorUtilityButton, parent=VerticalBox_270, widget=Spawn_Pistols, slot {class=VerticalBoxSlot}}
  TextBlock_53 {
   widgetClassPath=&TextBlock
   parent=Spawn_Pistols
   widget=TextBlock_53
   properties {
    justification=Center
    text=All Pistols
    font {
     bForceMonospaced=false
     bMaterialIsStencil=false
     fontMaterial=null
     fontObject=&Roboto
     letterSpacing=0
     monospacedWidth=1
     size=9
     skewAmount=0
     typefaceFontName=Regular
     outlineSettings {
      bApplyOutlineToDropShadows=false
      bMiteredCorners=false
      bSeparateFillAlpha=false
      outlineMaterial=null
      outlineSize=0
      outlineColor=<rgba 0 0 0 1>
     }
    }
   }
   slot {class=ButtonSlot, properties {horizontalAlignment=HAlign_Fill, verticalAlignment=VAlign_Fill}}
  }
  Spawn_Rifles {widgetClassPath=&EditorUtilityButton, parent=VerticalBox_270, widget=Spawn_Rifles, slot {class=VerticalBoxSlot}}
  TextBlock_52 {
   widgetClassPath=&TextBlock
   parent=Spawn_Rifles
   widget=TextBlock_52
   properties {
    justification=Center
    text=All Rifles
    font {
     bForceMonospaced=false
     bMaterialIsStencil=false
     fontMaterial=null
     fontObject=&Roboto
     letterSpacing=0
     monospacedWidth=1
     size=9
     skewAmount=0
     typefaceFontName=Regular
     outlineSettings {
      bApplyOutlineToDropShadows=false
      bMiteredCorners=false
      bSeparateFillAlpha=false
      outlineMaterial=null
      outlineSize=0
      outlineColor=<rgba 0 0 0 1>
     }
    }
   }
   slot {class=ButtonSlot, properties {horizontalAlignment=HAlign_Fill, verticalAlignment=VAlign_Fill}}
  }
  Spawn_Shotguns {widgetClassPath=&EditorUtilityButton, parent=VerticalBox_270, widget=Spawn_Shotguns, slot {class=VerticalBoxSlot}}
  TextBlock_54 {
   widgetClassPath=&TextBlock
   parent=Spawn_Shotguns
   widget=TextBlock_54
   properties {
    justification=Center
    text=All Shotguns
    font {
     bForceMonospaced=false
     bMaterialIsStencil=false
     fontMaterial=null
     fontObject=&Roboto
     letterSpacing=0
     monospacedWidth=1
     size=9
     skewAmount=0
     typefaceFontName=Regular
     outlineSettings {
      bApplyOutlineToDropShadows=false
      bMiteredCorners=false
      bSeparateFillAlpha=false
      outlineMaterial=null
      outlineSize=0
      outlineColor=<rgba 0 0 0 1>
     }
    }
   }
   slot {class=ButtonSlot, properties {horizontalAlignment=HAlign_Fill, verticalAlignment=VAlign_Fill}}
  }
  VerticalBox_2 {
   widgetClassPath=&VerticalBox
   parent=HorizontalBox_251
   widget=VerticalBox_2
   slot {class=HorizontalBoxSlot, properties {padding=<ltrb 1.25 0 0 0>, size {sizeRule=Fill, value=1}}}
  }
  TextBlock_47 {
   widgetClassPath=&TextBlock
   parent=VerticalBox_2
   widget=TextBlock_47
   properties {
    text=Items
    font {
     bForceMonospaced=false
     bMaterialIsStencil=false
     fontMaterial=null
     fontObject=&Roboto
     letterSpacing=0
     monospacedWidth=1
     size=9
     skewAmount=0
     typefaceFontName=Regular
     outlineSettings {
      bApplyOutlineToDropShadows=false
      bMiteredCorners=false
      bSeparateFillAlpha=false
      outlineMaterial=null
      outlineSize=0
      outlineColor=<rgba 0 0 0 1>
     }
    }
   }
   slot {
    class=VerticalBoxSlot
    properties {horizontalAlignment=HAlign_Center, verticalAlignment=VAlign_Center, padding=<ltrb 4 2 4 2>, size {sizeRule=Fill, value=1}}
   }
  }
  Spawn_Ingredients {
   widgetClassPath=&EditorUtilityButton
   parent=VerticalBox_2
   widget=Spawn_Ingredients
   slot {class=VerticalBoxSlot, properties {size {sizeRule=Fill, value=1}}}
  }
  Text {
   widgetClassPath=&TextBlock
   parent=Spawn_Ingredients
   widget=Text
   properties {
    justification=Center
    text=All Ingredients
    font {
     bForceMonospaced=false
     bMaterialIsStencil=false
     fontMaterial=null
     fontObject=&Roboto
     letterSpacing=0
     monospacedWidth=1
     size=9
     skewAmount=0
     typefaceFontName=Regular
     outlineSettings {
      bApplyOutlineToDropShadows=false
      bMiteredCorners=false
      bSeparateFillAlpha=false
      outlineMaterial=null
      outlineSize=0
      outlineColor=<rgba 0 0 0 1>
     }
    }
   }
   slot {class=ButtonSlot, properties {horizontalAlignment=HAlign_Fill}}
  }
  Spawn_Traps {
   widgetClassPath=&EditorUtilityButton
   parent=VerticalBox_2
   widget=Spawn_Traps
   slot {class=VerticalBoxSlot, properties {size {sizeRule=Fill, value=1}}}
  }
  Text_1 {
   widgetClassPath=&TextBlock
   parent=Spawn_Traps
   widget=Text_1
   properties {
    justification=Center
    text=All Traps
    font {
     bForceMonospaced=false
     bMaterialIsStencil=false
     fontMaterial=null
     fontObject=&Roboto
     letterSpacing=0
     monospacedWidth=1
     size=9
     skewAmount=0
     typefaceFontName=Regular
     outlineSettings {
      bApplyOutlineToDropShadows=false
      bMiteredCorners=false
      bSeparateFillAlpha=false
      outlineMaterial=null
      outlineSize=0
      outlineColor=<rgba 0 0 0 1>
     }
    }
   }
   slot {class=ButtonSlot, properties {horizontalAlignment=HAlign_Fill}}
  }
  TextBlock_48 {
   widgetClassPath=&TextBlock
   parent=VerticalBox_2
   widget=TextBlock_48
   properties {
    text=Consumables
    font {
     bForceMonospaced=false
     bMaterialIsStencil=false
     fontMaterial=null
     fontObject=&Roboto
     letterSpacing=0
     monospacedWidth=1
     size=9
     skewAmount=0
     typefaceFontName=Regular
     outlineSettings {
      bApplyOutlineToDropShadows=false
      bMiteredCorners=false
      bSeparateFillAlpha=false
      outlineMaterial=null
      outlineSize=0
      outlineColor=<rgba 0 0 0 1>
     }
    }
   }
   slot {
    class=VerticalBoxSlot
    properties {horizontalAlignment=HAlign_Center, verticalAlignment=VAlign_Center, padding=<ltrb 4 2 4 2>, size {sizeRule=Fill, value=1}}
   }
  }
  Spawn_Cosumables {
   widgetClassPath=&EditorUtilityButton
   parent=VerticalBox_2
   widget=Spawn_Cosumables
   slot {class=VerticalBoxSlot, properties {size {sizeRule=Fill, value=1}}}
  }
  TextBlock_56 {
   widgetClassPath=&TextBlock
   parent=Spawn_Cosumables
   widget=TextBlock_56
   properties {
    justification=Center
    text=All Consumables
    font {
     bForceMonospaced=false
     bMaterialIsStencil=false
     fontMaterial=null
     fontObject=&Roboto
     letterSpacing=0
     monospacedWidth=1
     size=9
     skewAmount=0
     typefaceFontName=Regular
     outlineSettings {
      bApplyOutlineToDropShadows=false
      bMiteredCorners=false
      bSeparateFillAlpha=false
      outlineMaterial=null
      outlineSize=0
      outlineColor=<rgba 0 0 0 1>
     }
    }
   }
   slot {class=ButtonSlot, properties {horizontalAlignment=HAlign_Fill}}
  }
  NPCCheats {
   widgetClassPath=&HorizontalBox
   parent=Spawn
   widget=NPCCheats
   slot {class=VerticalBoxSlot, properties {padding=<ltrb 0 0 0 2.5>}}
  }
  Spawn_NPCExpand {
   widgetClassPath=&EditorUtilityExpandableArea
   bIsVariable=true
   parent=NPCCheats
   widget=Spawn_NPCExpand
   slot {class=HorizontalBoxSlot, properties {size {sizeRule=Fill, value=1}}}
  }
  TextBlock_18 {
   widgetClassPath=&TextBlock
   namedSlotHost=Spawn_NPCExpand
   widget=TextBlock_18
   properties {
    text=NPCs
    toolTipText=Simple NPC spawning cheats.
    font {
     bForceMonospaced=false
     bMaterialIsStencil=false
     fontMaterial=null
     fontObject=&Roboto
     letterSpacing=0
     monospacedWidth=1
     size=9
     skewAmount=0
     typefaceFontName=Regular
     outlineSettings {
      bApplyOutlineToDropShadows=false
      bMiteredCorners=false
      bSeparateFillAlpha=false
      outlineMaterial=null
      outlineSize=0
      outlineColor=<rgba 0 0 0 1>
     }
    }
   }
  }
  VerticalBox_162 {widgetClassPath=&VerticalBox, namedSlotHost=Spawn_NPCExpand, widget=VerticalBox_162}
  HorizontalBox_110 {widgetClassPath=&HorizontalBox, parent=VerticalBox_162, widget=HorizontalBox_110, slot {class=VerticalBoxSlot}}
  NPC_Idle {
   widgetClassPath=&EditorUtilityButton
   parent=HorizontalBox_110
   widget=NPC_Idle
   properties {toolTipText=This spawns a very basic Unreal Pawn.}
   slot {class=HorizontalBoxSlot, properties {size {sizeRule=Fill, value=1}}}
  }
  TextBlock_43 {
   widgetClassPath=&TextBlock
   parent=NPC_Idle
   widget=TextBlock_43
   properties {
    justification=Center
    text=Spawn Idle
    font {
     bForceMonospaced=false
     bMaterialIsStencil=false
     fontMaterial=null
     fontObject=&Roboto
     letterSpacing=0
     monospacedWidth=1
     size=9
     skewAmount=0
     typefaceFontName=Regular
     outlineSettings {
      bApplyOutlineToDropShadows=false
      bMiteredCorners=false
      bSeparateFillAlpha=false
      outlineMaterial=null
      outlineSize=0
      outlineColor=<rgba 0 0 0 1>
     }
    }
   }
   slot {class=ButtonSlot, properties {horizontalAlignment=HAlign_Fill, verticalAlignment=VAlign_Fill}}
  }
  NPC_Bot {
   widgetClassPath=&EditorUtilityButton
   parent=HorizontalBox_110
   widget=NPC_Bot
   properties {toolTipText=Spawns a player bot with some really odd behaviors.}
   slot {class=HorizontalBoxSlot, properties {size {sizeRule=Fill, value=1}}}
  }
  TextBlock_44 {
   widgetClassPath=&TextBlock
   parent=NPC_Bot
   widget=TextBlock_44
   properties {
    justification=Center
    text=Spawn Bot
    font {
     bForceMonospaced=false
     bMaterialIsStencil=false
     fontMaterial=null
     fontObject=&Roboto
     letterSpacing=0
     monospacedWidth=1
     size=9
     skewAmount=0
     typefaceFontName=Regular
     outlineSettings {
      bApplyOutlineToDropShadows=false
      bMiteredCorners=false
      bSeparateFillAlpha=false
      outlineMaterial=null
      outlineSize=0
      outlineColor=<rgba 0 0 0 1>
     }
    }
   }
   slot {class=ButtonSlot, properties {horizontalAlignment=HAlign_Fill, verticalAlignment=VAlign_Fill}}
  }
  NPC_NPC {
   widgetClassPath=&EditorUtilityButton
   parent=HorizontalBox_110
   widget=NPC_NPC
   properties {toolTipText=Spawns a quest NPC.}
   slot {class=HorizontalBoxSlot, properties {size {sizeRule=Fill, value=1}}}
  }
  TextBlock_45 {
   widgetClassPath=&TextBlock
   parent=NPC_NPC
   widget=TextBlock_45
   properties {
    justification=Center
    text=Spawn NPC
    font {
     bForceMonospaced=false
     bMaterialIsStencil=false
     fontMaterial=null
     fontObject=&Roboto
     letterSpacing=0
     monospacedWidth=1
     size=9
     skewAmount=0
     typefaceFontName=Regular
     outlineSettings {
      bApplyOutlineToDropShadows=false
      bMiteredCorners=false
      bSeparateFillAlpha=false
      outlineMaterial=null
      outlineSize=0
      outlineColor=<rgba 0 0 0 1>
     }
    }
   }
   slot {class=ButtonSlot, properties {horizontalAlignment=HAlign_Fill, verticalAlignment=VAlign_Fill}}
  }
  Separator_4 {widgetClassPath=&HorizontalBox, parent=Spawn, widget=Separator_4, slot {class=VerticalBoxSlot}}
  Border_3 {
   widgetClassPath=&Border
   parent=Separator_4
   widget=Border_3
   slot {class=HorizontalBoxSlot, properties {size {sizeRule=Fill, value=1}}}
   properties=~
  }
 ]
}
